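# Physics asset for the accessories of the Pumpkin Spice female medium character body, covering a dangling pumpkin, a scraper and a back vine.
# The vine is an eight-body capsule chain hung from the pelvis with tight swing and twist limits.
# Pumpkin and scraper each hang from a small master body with free rotation against the pelvis.
# The pumpkin has limited swing and twist, while the scraper is locked to one swing axis.
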PhysicsAsset /BRCosmetics/Characters/Player/Female/Medium/Bodies/F_MED_Pumpkin_Spice/Meshes/F_MED_Pumpkin_Spice_Accessories_Physics
bodies {
 dyn_pumpkin4 [{shapeType=Sphere, radius=10, shapeName=dyn_pumpkin4_sphere, center=<0.5957 -0.1712 -6.2769>}]
 dyn_pumpkin_mstr [{shapeType=Capsule, length=0.505, radius=0.505}]
 dyn_scraper [{shapeType=Box, extentX=9.1254, extentY=4.9739, extentZ=1.2103, center=<0.4443 0.4249 0.0573>}]
 dyn_scraper_mstr [{shapeType=Box, extentX=1.01, extentY=1.01, extentZ=1.01}]
 dyn_vineBack1 [{shapeType=Capsule, length=3.3936, radius=1.0047, center=<3.2431 0.2847 0.3858>, rotation=<rot 90 0 0>}]
 dyn_vineBack2 [{shapeType=Capsule, length=11.2128, radius=1.8827, center=<7.6913 -0.221 -0.1082>, rotation=<rot 90 0 0>}]
 dyn_vineBack3 [{shapeType=Capsule, length=10.098, radius=1.227, center=<5.5421 0 0>, rotation=<rot 90 0 0>}]
 dyn_vineBack4 [{shapeType=Capsule, length=9.4183, radius=1.301, center=<6.0541 0 0>, rotation=<rot 90 0 0>}]
 dyn_vineBack5 [{shapeType=Capsule, length=3.2616, radius=0.9078, center=<2.2169 0 0>, rotation=<rot 90 0 0>}]
 dyn_vineBack6 [{shapeType=Capsule, length=7.2904, radius=1.9761, center=<6.1631 0 0>, rotation=<rot 90 0 0>}]
 dyn_vineBack7 [{shapeType=Capsule, length=0.9545, radius=1.6836, center=<1.8302 0 0>, rotation=<rot 90 0 0>}]
 dyn_vineBack8 [{shapeType=Capsule, length=2.1742, radius=1.9151, center=<0 0 -2.3439>}]
 pelvis [{shapeType=Capsule, length=11.3827, radius=18.4082, center=<-7.4017 -2.8533 0.8726>}]
}
constraints [
 {
  bone1Name=dyn_scraper
  bone2Name=dyn_scraper_mstr
  swing1LimitDegrees=10
  swing2LimitDegrees=45
  swing2Motion=Locked
  twistLimitDegrees=15
  twistMotion=Locked
 }
 {bone1Name=dyn_pumpkin4, bone2Name=dyn_pumpkin_mstr, swing1LimitDegrees=25, swing2LimitDegrees=25, twistLimitDegrees=5}
 {bone1Name=dyn_vineBack2, bone2Name=dyn_vineBack1, swing1LimitDegrees=15, swing2LimitDegrees=15, twistLimitDegrees=2}
 {bone1Name=dyn_vineBack3, bone2Name=dyn_vineBack2, swing1LimitDegrees=15, swing2LimitDegrees=15, twistLimitDegrees=5}
 {bone1Name=dyn_vineBack4, bone2Name=dyn_vineBack3, swing1LimitDegrees=15, swing2LimitDegrees=15, twistLimitDegrees=5}
 {bone1Name=dyn_vineBack5, bone2Name=dyn_vineBack4, swing1LimitDegrees=15, swing2LimitDegrees=15, twistLimitDegrees=5}
 {bone1Name=dyn_vineBack6, bone2Name=dyn_vineBack5, swing1LimitDegrees=15, swing2LimitDegrees=15, twistLimitDegrees=5}
 {bone1Name=dyn_vineBack7, bone2Name=dyn_vineBack6, swing1LimitDegrees=15, swing2LimitDegrees=15, twistLimitDegrees=5}
 {bone1Name=dyn_vineBack8, bone2Name=dyn_vineBack7, swing1LimitDegrees=15, swing2LimitDegrees=15, twistLimitDegrees=5}
 {
  bone1Name=dyn_pumpkin_mstr
  bone2Name=pelvis
  swing1LimitDegrees=45
  swing1Motion=Free
  swing2LimitDegrees=45
  swing2Motion=Free
  twistLimitDegrees=45
  twistMotion=Free
 }
 {bone1Name=dyn_vineBack1, bone2Name=pelvis, swing1LimitDegrees=15, swing2LimitDegrees=15, twistLimitDegrees=15}
 {
  bone1Name=dyn_scraper_mstr
  bone2Name=pelvis
  swing1LimitDegrees=45
  swing1Motion=Free
  swing2LimitDegrees=45
  swing2Motion=Free
  twistLimitDegrees=45
  twistMotion=Free
 }
]
