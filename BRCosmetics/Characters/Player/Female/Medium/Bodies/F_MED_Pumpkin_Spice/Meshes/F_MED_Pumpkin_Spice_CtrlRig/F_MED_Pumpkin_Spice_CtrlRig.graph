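# Control Rig for the F_MED_Pumpkin_Spice female body that adds procedural wave motion to some of its dangling costume bones.
# Execution runs four chained ChainHarmonics units in sequence, rooted at the back vine, pumpkin, left leg vine and right leg vine chains.
# Only Wave is enabled on each, with per-chain speed and frequency; pendulum and reach are off, and debug drawing is on.
# Skirt, pouch, string and other dyn_ bones are left untouched.
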ControlRigBlueprint /BRCosmetics/Characters/Player/Female/Medium/Bodies/F_MED_Pumpkin_Spice/Meshes/F_MED_Pumpkin_Spice_CtrlRig/F_MED_Pumpkin_Spice_CtrlRig
graphs [
 RigVMFunctionLibrary {links [], nodes []}
 RigVMModel {nodes [
   RigUnit_BeginExecution = &RigUnit_BeginExecution()
   /Script/ControlRig.RigUnit_ChainHarmonics(
    ChainRoot=dyn_vineBack1
    DrawWorldOffset=(Rotation=(X=0,Y=0,Z=0,W=1),Translation=(X=0,Y=0,Z=0),Scale3D=(X=1,Y=1,Z=1))
    Pendulum=(bEnabled=false,PendulumStiffness=2,PendulumGravity=(X=0,Y=0,Z=0),PendulumBlend=0.75,PendulumDrag=0.98,PendulumMinimum=0,PendulumMaximum=0.1,PendulumEase=Linear,UnwindAxis=(X=0,Y=1,Z=0),UnwindMinimum=0.2,UnwindMaximum=0.05)
    Reach=(bEnabled=false,ReachTarget=(X=10,Y=0,Z=10),ReachAxis=(X=1,Y=0,Z=0),ReachMinimum=0,ReachMaximum=-5,ReachEase=Linear)
    Speed=(X=0.75,Y=0.75,Z=0.75)
    Wave=(bEnabled=True,WaveFrequency=(X=15,Y=15,Z=-15),WaveAmplitude=(X=0.5,Y=0.5,Z=0.5),WaveOffset=(X=0.5,Y=0.5,Z=0.5),WaveNoise=(X=0,Y=0.5,Z=0.5),WaveMinimum=0,WaveMaximum=1,WaveEase=Linear)
    WaveCurve=(EditorCurveData=(Keys=((Value=0.25),(Time=1.0107,Value=0.6084)),DefaultValue=340282346638528859811704183484516925440,PreInfinityExtrap=RCCE_Constant,PostInfinityExtrap=RCCE_Constant))
    bDrawDebug=True
    ExecutePin=@RigUnit_BeginExecution.ExecutePin
   ) {is=struct RigVMUnitNode}
   > &RigUnit_ChainHarmonics(
    ChainRoot=dyn_pumpkin1
    DrawWorldOffset=(Rotation=(X=0,Y=0,Z=0,W=1),Translation=(X=0,Y=0,Z=0),Scale3D=(X=1,Y=1,Z=1))
    Pendulum=(bEnabled=false,PendulumStiffness=2,PendulumGravity=(X=0,Y=0,Z=0),PendulumBlend=0.75,PendulumDrag=0.98,PendulumMinimum=0,PendulumMaximum=0.1,PendulumEase=Linear,UnwindAxis=(X=0,Y=1,Z=0),UnwindMinimum=0.2,UnwindMaximum=0.05)
    Reach=(bEnabled=false,ReachTarget=(X=10,Y=0,Z=10),ReachAxis=(X=1,Y=0,Z=0),ReachMinimum=0,ReachMaximum=-5,ReachEase=Linear)
    Speed=(X=0.6,Y=0.6,Z=0.6)
    Wave=(bEnabled=True,WaveFrequency=(X=10,Y=10,Z=-10),WaveAmplitude=(X=0.25,Y=0.25,Z=0.25),WaveOffset=(X=-0.5,Y=0.5,Z=0.5),WaveNoise=(X=0,Y=0.5,Z=0.5),WaveMinimum=0,WaveMaximum=1,WaveEase=Linear)
    WaveCurve=(EditorCurveData=(Keys=((Value=0.25),(Time=1.0107,Value=0.6084)),DefaultValue=340282346638528859811704183484516925440,PreInfinityExtrap=RCCE_Constant,PostInfinityExtrap=RCCE_Constant))
    bDrawDebug=True
   )
   > &RigUnit_ChainHarmonics(
    ChainRoot=dyn_legVine1_l
    DrawWorldOffset=(Rotation=(X=0,Y=0,Z=0,W=1),Translation=(X=0,Y=0,Z=0),Scale3D=(X=1,Y=1,Z=1))
    Pendulum=(bEnabled=false,PendulumStiffness=2,PendulumGravity=(X=0,Y=0,Z=0),PendulumBlend=0.75,PendulumDrag=0.98,PendulumMinimum=0,PendulumMaximum=0.1,PendulumEase=Linear,UnwindAxis=(X=0,Y=1,Z=0),UnwindMinimum=0.2,UnwindMaximum=0.05)
    Reach=(bEnabled=false,ReachTarget=(X=10,Y=0,Z=10),ReachAxis=(X=1,Y=0,Z=0),ReachMinimum=0,ReachMaximum=-5,ReachEase=Linear)
    Speed=(X=0.6,Y=0.6,Z=0.6)
    Wave=(bEnabled=True,WaveFrequency=(X=50,Y=50,Z=50),WaveAmplitude=(X=0.1,Y=-0.1,Z=-0.1),WaveOffset=(X=0,Y=0,Z=0),WaveNoise=(X=0.5,Y=0.5,Z=0.5),WaveMinimum=0,WaveMaximum=1,WaveEase=Linear)
    WaveCurve=(EditorCurveData=(Keys=((Value=0.25),(Time=1.0107,Value=0.6084)),DefaultValue=340282346638528859811704183484516925440,PreInfinityExtrap=RCCE_Constant,PostInfinityExtrap=RCCE_Constant))
    bDrawDebug=True
   )
   > &RigUnit_ChainHarmonics(
    ChainRoot=dyn_legVine1_r
    DrawWorldOffset=(Rotation=(X=0,Y=0,Z=0,W=1),Translation=(X=0,Y=0,Z=0),Scale3D=(X=1,Y=1,Z=1))
    Pendulum=(bEnabled=false,PendulumStiffness=2,PendulumGravity=(X=0,Y=0,Z=0),PendulumBlend=0.75,PendulumDrag=0.98,PendulumMinimum=0,PendulumMaximum=0.1,PendulumEase=Linear,UnwindAxis=(X=0,Y=1,Z=0),UnwindMinimum=0.2,UnwindMaximum=0.05)
    Reach=(bEnabled=false,ReachTarget=(X=10,Y=0,Z=10),ReachAxis=(X=1,Y=0,Z=0),ReachMinimum=0,ReachMaximum=-5,ReachEase=Linear)
    Speed=(X=0.6,Y=0.6,Z=0.6)
    Wave=(bEnabled=True,WaveFrequency=(X=50,Y=50,Z=50),WaveAmplitude=(X=0.1,Y=-0.1,Z=-0.1),WaveOffset=(X=0,Y=0,Z=0),WaveNoise=(X=0.5,Y=0.5,Z=0.5),WaveMinimum=0,WaveMaximum=1,WaveEase=Linear)
    WaveCurve=(EditorCurveData=(Keys=((Value=0.25),(Time=1.0107,Value=0.6084)),DefaultValue=340282346638528859811704183484516925440,PreInfinityExtrap=RCCE_Constant,PostInfinityExtrap=RCCE_Constant))
    bDrawDebug=True
   )
  ]}
]
hierarchy [
 root {loc=<0 0 0>, rot=<0 0 0 1>}
 pelvis {parent=root, loc=<0 -0.5301 98.0571>, rot=<0.0127 0.707 -0.0127 -0.707>}
 spine_01 {parent=pelvis, loc=<6.1208 -0.9328 0>, rot=<0 0 0.034 -0.9994>}
 spine_02 {parent=spine_01, loc=<6.5996 -0.2095 0>, rot=<0 0 0.0238 -0.9997>}
 spine_03 {parent=spine_02, loc=<6.4914 -0.1237 0>, rot=<0 0 0.0149 -0.9999>}
 spine_04 {parent=spine_03, loc=<6.9598 0.0794 0>, rot=<0 0 -0.0132 -0.9999>}
 spine_05 {parent=spine_04, loc=<6.9987 0 0>, rot=<0 0 -0.0015 -1>}
 clavicle_l {parent=spine_05, loc=<5.4831 -0.1053 -3.2946>, rot=<0.0465 -0.7289 -0.009 -0.6829>}
 upperarm_l {parent=clavicle_l, loc=<9.8658 -1.2136 -0.4105>, rot=<0.0164 -0.3245 -0.0269 -0.9454>}
 lowerarm_l {parent=upperarm_l, loc=<26.465 -0.8345 0.259>, rot=<0.0036 0.0161 -0.2298 -0.9731>}
 hand_l {parent=lowerarm_l, loc=<22.7355 -0.3203 -0.6105>, rot=<-0.6784 -0.0223 0.0735 -0.7307>}
 index_metacarpal_l {parent=hand_l, loc=<1.6535 0.0183 -2.96>, rot=<0.0158 -0.0124 -0.0173 -0.9996>}
 index_01_l {parent=index_metacarpal_l, loc=<8.3008 0.0357 -0.0524>, rot=<-0.0044 0.0485 0.0924 -0.9945>}
 index_02_l {parent=index_01_l, loc=<3.3587 -0.3003 0.0452>, rot=<-0.002 0.0027 0.3225 -0.9466>}
 index_03_l {parent=index_02_l, loc=<2.5578 -0.123 -0.0252>, rot=<-0.002 0.0067 0.2558 -0.9667>}
 middle_metacarpal_l {parent=hand_l, loc=<2.1856 -0.1725 -0.5984>, rot=<0.006 0.0168 -0.0329 -0.9993>}
 middle_01_l {parent=middle_metacarpal_l, loc=<7.3275 -0.2447 0.0129>, rot=<-0.0208 0.0042 0.0858 -0.9961>}
 middle_02_l {parent=middle_01_l, loc=<4.0123 0.1401 0.1175>, rot=<-0.0218 0.0258 0.3512 -0.9357>}
 middle_03_l {parent=middle_02_l, loc=<2.6077 0.0534 0.1373>, rot=<0.0082 0.0188 0.1708 -0.9851>}
 pinky_metacarpal_l {parent=hand_l, loc=<1.7616 -0.9753 3.3814>, rot=<-0.0397 0.0814 -0.069 -0.9935>}
 pinky_01_l {parent=pinky_metacarpal_l, loc=<7.1709 -0.4831 -0.1094>, rot=<-0.0654 -0.0035 0.2242 -0.9723>}
 pinky_02_l {parent=pinky_01_l, loc=<2.8274 0.0893 -0.1043>, rot=<0.0037 0.0011 0.2854 -0.9584>}
 pinky_03_l {parent=pinky_02_l, loc=<2.2046 0.0308 -0.0078>, rot=<-0.0002 0.0098 0.1445 -0.9895>}
 ring_metacarpal_l {parent=hand_l, loc=<2.2074 -0.3625 1.3638>, rot=<-0.0284 0.0722 -0.0458 -0.9959>}
 ring_01_l {parent=ring_metacarpal_l, loc=<7.0295 -0.6197 -0.1446>, rot=<-0.0489 -0.0252 0.1892 -0.9804>}
 ring_02_l {parent=ring_01_l, loc=<3.6174 0.4647 -0.1027>, rot=<0 0 0.2806 -0.9598>}
 ring_03_l {parent=ring_02_l, loc=<2.4152 0.1141 0.0199>, rot=<0 0 0.1189 -0.9929>}
 thumb_01_l {parent=hand_l, loc=<1.3118 -0.3191 -2.459>, rot=<0.6546 -0.4042 0.0664 -0.6354>}
 thumb_02_l {parent=thumb_01_l, loc=<5.9336 -0.0448 0.0401>, rot=<-0.0037 0.0213 0.298 -0.9543>}
 thumb_03_l {parent=thumb_02_l, loc=<2.2223 0.3682 -0.0907>, rot=<0.0118 -0.013 0.0066 -0.9998>}
 lowerarm_twist_01_l {parent=lowerarm_l, loc=<18.316 0 0>, rot=<0 0 0 1>}
 lowerarm_twist_02_l {parent=lowerarm_l, loc=<10.6707 0 0>, rot=<0 0 0 1>}
 upperarm_twist_01_l {parent=upperarm_l, loc=<7.2135 -0.1134 0>, rot=<0 0 0 -1>}
 upperarm_twist_02_l {parent=upperarm_l, loc=<19.7998 -0.1134 0>, rot=<0 0 0 -1>}
 clavicle_r {parent=spine_05, loc=<5.4831 -0.1053 3.2946>, rot=<0.7289 0.0465 -0.6829 0.009>}
 upperarm_r {parent=clavicle_r, loc=<-9.8658 1.2136 0.4104>, rot=<0.0164 -0.3245 -0.0269 -0.9454>}
 lowerarm_r {parent=upperarm_r, loc=<-26.4651 0.8345 -0.259>, rot=<0.0036 0.0161 -0.2298 -0.9731>}
 hand_r {parent=lowerarm_r, loc=<-22.7356 0.3203 0.6106>, rot=<-0.6784 -0.0223 0.0735 -0.7307>}
 index_metacarpal_r {parent=hand_r, loc=<-1.6534 -0.0181 2.9601>, rot=<0.0158 -0.0124 -0.0173 -0.9996>}
 index_01_r {parent=index_metacarpal_r, loc=<-8.3008 -0.0357 0.0524>, rot=<-0.0044 0.0485 0.0924 -0.9945>}
 index_02_r {parent=index_01_r, loc=<-3.3589 0.3003 -0.0453>, rot=<-0.002 0.0027 0.3225 -0.9466>}
 index_03_r {parent=index_02_r, loc=<-2.5578 0.123 0.0252>, rot=<-0.002 0.0067 0.2558 -0.9667>}
 middle_metacarpal_r {parent=hand_r, loc=<-2.1856 0.1726 0.5984>, rot=<0.006 0.0168 -0.0329 -0.9993>}
 middle_01_r {parent=middle_metacarpal_r, loc=<-7.3273 0.2447 -0.0129>, rot=<-0.0208 0.0042 0.0858 -0.9961>}
 middle_02_r {parent=middle_01_r, loc=<-4.0124 -0.14 -0.1175>, rot=<-0.0218 0.0258 0.3512 -0.9357>}
 middle_03_r {parent=middle_02_r, loc=<-2.6077 -0.0533 -0.1373>, rot=<0.0082 0.0188 0.1708 -0.9851>}
 pinky_metacarpal_r {parent=hand_r, loc=<-1.7617 0.9753 -3.3814>, rot=<-0.0397 0.0814 -0.069 -0.9935>}
 pinky_01_r {parent=pinky_metacarpal_r, loc=<-7.1707 0.4832 0.1094>, rot=<-0.0654 -0.0035 0.2242 -0.9723>}
 pinky_02_r {parent=pinky_01_r, loc=<-2.8274 -0.0892 0.1043>, rot=<0.0037 0.0011 0.2854 -0.9584>}
 pinky_03_r {parent=pinky_02_r, loc=<-2.2045 -0.031 0.0078>, rot=<-0.0002 0.0098 0.1445 -0.9895>}
 ring_metacarpal_r {parent=hand_r, loc=<-2.2073 0.3627 -1.3638>, rot=<-0.0284 0.0722 -0.0458 -0.9959>}
 ring_01_r {parent=ring_metacarpal_r, loc=<-7.0297 0.6197 0.1446>, rot=<-0.0489 -0.0252 0.1892 -0.9804>}
 ring_02_r {parent=ring_01_r, loc=<-3.6172 -0.4647 0.1027>, rot=<0 0 0.2806 -0.9598>}
 ring_03_r {parent=ring_02_r, loc=<-2.4152 -0.1141 -0.0199>, rot=<0 0 0.1189 -0.9929>}
 thumb_01_r {parent=hand_r, loc=<-1.3119 0.3192 2.459>, rot=<0.6546 -0.4042 0.0664 -0.6354>}
 thumb_02_r {parent=thumb_01_r, loc=<-5.9336 0.0448 -0.0401>, rot=<-0.0037 0.0213 0.298 -0.9543>}
 thumb_03_r {parent=thumb_02_r, loc=<-2.2223 -0.3682 0.0907>, rot=<0.0118 -0.013 0.0066 -0.9998>}
 lowerarm_twist_01_r {parent=lowerarm_r, loc=<-18.3161 0 0>, rot=<0 0 0 1>}
 lowerarm_twist_02_r {parent=lowerarm_r, loc=<-10.6707 0 0>, rot=<0 0 0 1>}
 upperarm_twist_01_r {parent=upperarm_r, loc=<-7.2135 0.1134 0>, rot=<0 0 0 1>}
 upperarm_twist_02_r {parent=upperarm_r, loc=<-19.7998 0.1134 0>, rot=<0 0 0 1>}
 neck_01 {parent=spine_05, loc=<11.3138 0.1529 0>, rot=<0 0 -0.2003 -0.9797>}
 neck_02 {parent=neck_01, loc=<4.5627 0 0>, rot=<0 0 0.0504 -0.9987>}
 head {parent=neck_02, loc=<5.3914 0 0>, rot=<0 0 0.115 -0.9934>}
 faceAttach {parent=head, loc=<-3.337 -0.4556 0.0011>, rot=<0.0082 -0.7071 -0.0082 -0.7071>, scale=<0.89 0.89 0.89>}
 C_jaw {parent=faceAttach, loc=<-0.0138 3.0827 1.1287>, rot=<0 0 0 1>}
 dyn_pumpkinCharm_mstr {parent=neck_01, loc=<0.7073 5.0482 -0.0537>, rot=<-0.2513 -0.6609 -0.6609 -0.2513>}
 dyn_pumpkinCharm {parent=dyn_pumpkinCharm_mstr, loc=<0 0 0>, rot=<0 0 0 1>}
 dyn_string_mstr {parent=spine_05, loc=<-1.4088 11.9768 -0.264>, rot=<0.2283 0.6232 0.724 0.188>}
 dyn_loop1_l {parent=dyn_string_mstr, loc=<0 0 0>, rot=<-0.1421 0.0315 0.2153 -0.9656>}
 dyn_loop2_l {parent=dyn_loop1_l, loc=<3.8345 0 0>, rot=<-0.0275 -0.1861 0.1435 -0.9716>}
 dyn_string1_l {parent=dyn_string_mstr, loc=<0 0 0>, rot=<-0.0951 0.0603 0.0869 -0.9898>}
 dyn_string2_l {parent=dyn_string1_l, loc=<4.5123 0 0>, rot=<-0.0139 -0.2238 0.0894 -0.9704>}
 dyn_string1_r {parent=dyn_string_mstr, loc=<0 0 0>, rot=<0 0 0 1>}
 dyn_string2_r {parent=dyn_string1_r, loc=<5.6649 0 0>, rot=<-0.0721 -0.2566 0.0902 -0.9596>}
 dyn_loop1_r {parent=dyn_string_mstr, loc=<0 0 0>, rot=<0.0364 0.1647 -0.1423 -0.9753>}
 dyn_loop2_r {parent=dyn_loop1_r, loc=<3.5502 0 0>, rot=<0 -0.2913 0 -0.9566>}
 dyn_pouch_mstr_l {parent=spine_02, loc=<6.434 -1.1187 -9.173>, rot=<0.0872 -0.9751 0.1456 -0.1428>}
 dyn_pouch_l {parent=dyn_pouch_mstr_l, loc=<0 0 0>, rot=<0 0 0 1>}
 dyn_pouch_mstr_r {parent=spine_02, loc=<3.8621 6.526 8.3726>, rot=<0.1768 0.3156 0.9322 -0.0083>}
 dyn_pouch_r {parent=dyn_pouch_mstr_r, loc=<0 0 0>, rot=<0 0 0 1>}
 dyn_pumpkin_mstr {parent=spine_02, loc=<4.5603 -4.5744 6.8139>, rot=<0.7718 -0.2039 0.4436 -0.4074>}
 dyn_pumpkin1 {parent=dyn_pumpkin_mstr, loc=<0 0 0>, rot=<0 0 0 1>}
 dyn_pumpkin2 {parent=dyn_pumpkin1, loc=<7.6404 0 0>, rot=<0.3693 -0.1795 -0.6336 -0.6557>}
 dyn_pumpkin3 {parent=dyn_pumpkin2, loc=<8.1588 0 0>, rot=<0.0428 -0.8198 -0.4208 -0.3861>}
 dyn_pumpkin4 {parent=dyn_pumpkin3, loc=<3.8236 -4.0191 0.4618>, rot=<-0.3775 -0.9233 -0.0269 0.0657>}
 dyn_scraper_mstr {parent=spine_01, loc=<6.9416 9.6691 -4.7263>, rot=<0.0222 -0.8346 -0.5502 0.0127>}
 dyn_scraper {parent=dyn_scraper_mstr, loc=<0 0 0>, rot=<0 0 0 1>}
 dyn_vineBack1 {parent=spine_01, loc=<4.4981 -8.0472 1.7065>, rot=<0.1321 -0.4775 0.2045 -0.8443>}
 dyn_vineBack2 {parent=dyn_vineBack1, loc=<7.8891 0 0>, rot=<0.4529 -0.7136 0.031 -0.5335>}
 dyn_vineBack3 {parent=dyn_vineBack2, loc=<14.9166 0 0>, rot=<-0.1761 0.7147 -0.2524 -0.6281>}
 dyn_vineBack4 {parent=dyn_vineBack3, loc=<11.9083 0 0>, rot=<-0.1086 -0.7808 0.0014 -0.6153>}
 dyn_vineBack5 {parent=dyn_vineBack4, loc=<13.0633 0 0>, rot=<-0.1263 0.5396 0.2978 -0.7773>}
 dyn_vineBack6 {parent=dyn_vineBack5, loc=<5.5038 0 0>, rot=<0.1235 -0.5671 -0.3031 -0.7559>}
 dyn_vineBack7 {parent=dyn_vineBack6, loc=<12.6631 0 0>, rot=<0.0111 -0.0501 0.3757 -0.9253>}
 dyn_vineBack8 {parent=dyn_vineBack7, loc=<3.9748 0 0>, rot=<-0.2339 0.75 -0.1842 -0.5906>}
 dyn_pumpkinTools_mstr {parent=spine_01, loc=<5.2231 9.2244 5.186>, rot=<-0.0088 0.5476 0.8366 -0.0135>}
 dyn_pumpkinTools {parent=dyn_pumpkinTools_mstr, loc=<0 0 0>, rot=<0 0 0 1>}
 dyn_spoonLoop {parent=spine_01, loc=<5.5493 7.6977 -6.7829>, rot=<-0.0969 -0.8826 -0.2626 -0.3777>}
 dyn_spoon_mstr {parent=dyn_spoonLoop, loc=<3.7613 0 0>, rot=<0.2973 0.1556 -0.0755 -0.939>}
 dyn_spoon {parent=dyn_spoon_mstr, loc=<0 0 0>, rot=<0 0 0 1>}
 thigh_l {parent=pelvis, loc=<-1.0855 -0.2109 -8.7053>, rot=<0.0009 0.0355 0.0155 -0.9993>}
 calf_l {parent=thigh_l, loc=<-42.5284 -0.0019 -0.7297>, rot=<0.0006 -0.0161 -0.0391 -0.9991>}
 calf_twist_01_l {parent=calf_l, loc=<-35.8483 -1.9197 -0.5794>, rot=<0.0237 0.0095 -0.0035 -0.9997>}
 calf_twist_02_l {parent=calf_l, loc=<-19.1546 -1.7943 -0.2638>, rot=<0.0237 0.0095 -0.0035 -0.9997>}
 foot_l {parent=calf_l, loc=<-46.7401 -2.5014 -0.592>, rot=<0.0427 -0.0238 0.0353 -0.9982>}
 ball_l {parent=foot_l, loc=<-3.5658 10.8066 -0.1544>, rot=<0.0001 -0.0001 -0.7063 -0.7079>}
 dyn_legVine1_l {parent=calf_l, loc=<-39.6883 2.0694 -1.2984>, rot=<0.544 -0.464 0.4214 -0.5578>}
 dyn_legVine2_l {parent=dyn_legVine1_l, loc=<10.5003 0 0>, rot=<0.2592 -0.0062 0.9485 -0.182>}
 dyn_legVine3_l {parent=dyn_legVine2_l, loc=<9.0786 0 0>, rot=<0.4148 0.0974 0.812 -0.399>}
 dyn_legVine4_l {parent=dyn_legVine3_l, loc=<9.374 0 0>, rot=<0.5168 0.0226 0.5586 -0.6484>}
 dyn_legVine5_l {parent=dyn_legVine4_l, loc=<11.9393 0 0>, rot=<0.1448 -0.3373 0.7888 -0.4931>}
 dyn_legVine6_l {parent=dyn_legVine5_l, loc=<11.3142 0 0>, rot=<-0.0486 0.1887 0.8439 -0.4999>}
 dyn_legVine7_l {parent=dyn_legVine6_l, loc=<11.9472 0 0>, rot=<0.4866 0.0999 0.7788 -0.3829>}
 thigh_twist_01_l {parent=thigh_l, loc=<-27.952 0 0>, rot=<0.0013 0 0 -1>}
 thigh_r {parent=pelvis, loc=<-1.0853 -0.2109 8.7053>, rot=<0.0355 -0.0009 0.9993 0.0155>}
 calf_r {parent=thigh_r, loc=<42.5286 0.0019 0.7297>, rot=<0.0006 -0.0161 -0.0391 -0.9991>}
 calf_twist_01_r {parent=calf_r, loc=<35.8485 1.9197 0.5794>, rot=<0.0237 0.0095 -0.0035 -0.9997>}
 calf_twist_02_r {parent=calf_r, loc=<19.1546 1.7943 0.2638>, rot=<0.0237 0.0095 -0.0035 -0.9997>}
 foot_r {parent=calf_r, loc=<46.7403 2.5014 0.592>, rot=<0.0427 -0.0238 0.0353 -0.9982>}
 ball_r {parent=foot_r, loc=<3.5658 -10.8066 0.1544>, rot=<0.0001 -0.0001 -0.7063 -0.7079>}
 dyn_legVine1_r {parent=calf_r, loc=<46.2934 6.1285 -1.9871>, rot=<0.3892 -0.0653 0.9158 -0.0739>}
 dyn_legVine2_r {parent=dyn_legVine1_r, loc=<7.4672 0 0>, rot=<-0.3996 -0.5698 -0.2591 -0.6697>}
 dyn_legVine3_r {parent=dyn_legVine2_r, loc=<8.5856 0 0>, rot=<-0.0048 0.308 0.7987 -0.5169>}
 dyn_legVine4_r {parent=dyn_legVine3_r, loc=<12.1448 0 0>, rot=<-0.1342 -0.7856 0.5588 0.2294>}
 dyn_legVine5_r {parent=dyn_legVine4_r, loc=<11.5481 0 0>, rot=<-0.2823 -0.4542 0.6028 -0.5922>}
 thigh_twist_01_r {parent=thigh_r, loc=<27.952 0 0>, rot=<0.0013 0 0 -1>}
 dyn_skirtFront_mstr {parent=pelvis, loc=<6.2525 10.5348 0.1028>, rot=<-0.226 -0.67 -0.67 -0.226>}
 dyn_skirtFront1 {parent=dyn_skirtFront_mstr, loc=<0 0 0>, rot=<0.0034 -0.0267 -0.0023 -0.9996>}
 dyn_skirtFront2 {parent=dyn_skirtFront1, loc=<14.2663 0 0>, rot=<-0.0343 -0.322 0.0081 -0.9461>}
 dyn_skirtFront1_mstr_l {parent=pelvis, loc=<6.9399 7.7039 -9.5981>, rot=<-0.1441 -0.8556 -0.4182 -0.2688>}
 dyn_skirtFront1_l {parent=dyn_skirtFront1_mstr_l, loc=<0 0 0>, rot=<-0.0037 -0.0276 0.0023 -0.9996>}
 dyn_skirtFront2_l {parent=dyn_skirtFront1_l, loc=<14.6728 0 0>, rot=<0.0348 -0.2854 -0.0127 -0.9577>}
 dyn_skirtSide1_mstr_l {parent=pelvis, loc=<7.0667 -1.2534 -13.5475>, rot=<0.0466 -0.944 0.0225 -0.3259>}
 dyn_skirtSide1_l {parent=dyn_skirtSide1_mstr_l, loc=<0 0 0>, rot=<-0.0925 -0.0257 0.0019 -0.9954>}
 dyn_skirtSide2_l {parent=dyn_skirtSide1_l, loc=<15.1963 0 0>, rot=<0.066 -0.3101 -0.0155 -0.9483>}
 dyn_skirtBack1_mstr_l {parent=pelvis, loc=<7.6492 -9.1167 -8.8861>, rot=<0.1151 -0.8615 0.4182 -0.2639>}
 dyn_skirtBack1_l {parent=dyn_skirtBack1_mstr_l, loc=<0 0 0>, rot=<0.0002 -0.0282 -0.0001 -0.9996>}
 dyn_skirtBack2_l {parent=dyn_skirtBack1_l, loc=<14.8561 0 0>, rot=<0.0006 -0.2614 0.0114 -0.9652>}
 dyn_skirtBack_mstr {parent=pelvis, loc=<7.5594 -12.3186 0.0998>, rot=<0.2001 -0.6799 0.6762 -0.2013>}
 dyn_skirtBack1 {parent=dyn_skirtBack_mstr, loc=<0 0 0>, rot=<-0.0013 -0.0291 0.0009 -0.9996>}
 dyn_skirtBack2 {parent=dyn_skirtBack1, loc=<14.5712 0 0>, rot=<0.015 -0.2521 0.0036 -0.9676>}
 dyn_skirtBack1_mstr_r {parent=pelvis, loc=<7.3164 -8.6605 9.2955>, rot=<0.2796 -0.4391 0.8434 -0.1333>}
 dyn_skirtBack1_r {parent=dyn_skirtBack1_mstr_r, loc=<0 0 0>, rot=<-0.0034 -0.0268 0.0025 -0.9996>}
 dyn_skirtBack2_r {parent=dyn_skirtBack1_r, loc=<14.8087 0 0>, rot=<0.0332 -0.2983 -0.026 -0.9535>}
 dyn_skirtSide1_mstr_r {parent=pelvis, loc=<7.2761 -3.0647 13.6035>, rot=<0.3046 -0.1212 0.9416 -0.077>}
 dyn_skirtSide1_r {parent=dyn_skirtSide1_mstr_r, loc=<0 0 0>, rot=<0.092 -0.0265 -0.0013 -0.9954>}
 dyn_skirtSide2_r {parent=dyn_skirtSide1_r, loc=<14.9865 0 0>, rot=<-0.0642 -0.2985 -0.0126 -0.9522>}
 dyn_skirtFront1_mstr_r {parent=pelvis, loc=<6.8413 7.2801 9.9231>, rot=<0.2619 0.4096 0.863 0.1373>}
 dyn_skirtFront1_r {parent=dyn_skirtFront1_mstr_r, loc=<0 0 0>, rot=<0.0099 -0.0273 -0.006 -0.9996>}
 dyn_skirtFront2_r {parent=dyn_skirtFront1_r, loc=<14.8638 0 0>, rot=<-0.091 -0.2722 0.0327 -0.9574>}
]
