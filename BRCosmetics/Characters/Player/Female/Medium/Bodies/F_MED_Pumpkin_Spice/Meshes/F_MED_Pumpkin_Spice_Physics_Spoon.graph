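# Physics asset for the Pumpkin Spice female medium body cosmetic that simulates its dangling spoon bone.
# The spoon uses a sphere and capsule shape, jointed to a small box master body with narrow swing and very tight twist limits.
# Capsules on the pelvis and left lower arm add body volumes around the spoon.
# The master and forearm bodies each link to the pelvis with free swing and twist.
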PhysicsAsset /BRCosmetics/Characters/Player/Female/Medium/Bodies/F_MED_Pumpkin_Spice/Meshes/F_MED_Pumpkin_Spice_Physics_Spoon
bodies {
 dyn_spoon [
  {shapeType=Sphere, radius=2.75, shapeName=dyn_spoon_sphere, center=<13.2487 0 -0.7285>}
  {shapeType=Capsule, length=10.35, radius=1.5, shapeName=dyn_spoon_capsule, center=<3.5864 0.3594 -0.9039>, rotation=<rot 80 0.0003 -0.0004>}
 ]
 dyn_spoon_mstr [{shapeType=Box, extentX=1.01, extentY=1.01, extentZ=1.01}]
 lowerarm_l [{shapeType=Capsule, length=19.755, radius=4.25, center=<4.1494 0 0>, rotation=<rot -90 0 0>}]
 pelvis [{shapeType=Capsule, length=13.3142, radius=17.6467, center=<-8.7703 -0.0083 0.1269>}]
}
constraints [
 {
  bone1Name=dyn_spoon_mstr
  bone2Name=pelvis
  swing1LimitDegrees=45
  swing1Motion=Free
  swing2LimitDegrees=45
  swing2Motion=Free
  twistLimitDegrees=45
  twistMotion=Free
 }
 {bone1Name=dyn_spoon, bone2Name=dyn_spoon_mstr, swing1LimitDegrees=30, swing2LimitDegrees=25, twistLimitDegrees=5}
 {
  bone1Name=lowerarm_l
  bone2Name=pelvis
  swing1LimitDegrees=45
  swing1Motion=Free
  swing2LimitDegrees=45
  swing2Motion=Free
  twistLimitDegrees=45
  twistMotion=Free
 }
]
