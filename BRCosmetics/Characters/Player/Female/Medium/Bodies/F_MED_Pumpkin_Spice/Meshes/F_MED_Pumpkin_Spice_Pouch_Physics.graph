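# Physics asset for the left and right pouches on the F_MED_Pumpkin_Spice female player body cosmetic.
# Each pouch bone carries an offset capsule body and is constrained to a small box body on its own master bone.
# Both constraints share the same angular limits, tight on swing1 and twist and wider on swing2, so each pouch swings mainly about one axis.
# The two sides have no constraint between them, and the right capsule is slightly larger than the left.
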PhysicsAsset /BRCosmetics/Characters/Player/Female/Medium/Bodies/F_MED_Pumpkin_Spice/Meshes/F_MED_Pumpkin_Spice_Pouch_Physics
bodies {
 dyn_pouch_l [{shapeType=Capsule, length=10.5, radius=3, shapeName=dyn_pouch_l_capsule, center=<6.3516 0.5618 2.1284>, rotation=<rot 90 0 0>}]
 dyn_pouch_mstr_l [{shapeType=Box, extentX=1.01, extentY=1.01, extentZ=1.01}]
 dyn_pouch_mstr_r [{shapeType=Box, extentX=1.01, extentY=1.01, extentZ=1.01}]
 dyn_pouch_r [{shapeType=Capsule, length=11.5, radius=3.5, shapeName=dyn_pouch_r_capsule, center=<7.7738 0.1297 2.7273>, rotation=<rot 90 0 0>}]
}
constraints [
 {bone1Name=dyn_pouch_l, bone2Name=dyn_pouch_mstr_l, swing1LimitDegrees=5, swing2LimitDegrees=30, twistLimitDegrees=5}
 {bone1Name=dyn_pouch_r, bone2Name=dyn_pouch_mstr_r, swing1LimitDegrees=5, swing2LimitDegrees=30, twistLimitDegrees=5}
]
