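# Physics asset for the skirt of the Pumpkin Spice medium female cosmetic body, named as the crouch variant.
# Seven front, side and back chains each link a tiny master box to two flat panel boxes through tightly limited angular constraints, slightly looser at the root joint.
# Capsules on both thigh twist bones provide leg collision shapes, and some lower front and right-side panels are widened.
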PhysicsAsset /BRCosmetics/Characters/Player/Female/Medium/Bodies/F_MED_Pumpkin_Spice/Meshes/F_MED_Pumpkin_Spice_SkirtCrouch_Physics
bodies {
 dyn_skirtBack1 [{shapeType=Box, extentX=15, extentY=10, extentZ=3, center=<7.5 0 0>}]
 dyn_skirtBack1_l [{shapeType=Box, extentX=15, extentY=10, extentZ=3, center=<7.5 0 0>}]
 dyn_skirtBack1_mstr_l [{shapeType=Box, extentX=1.01, extentY=1.01, extentZ=1.01}]
 dyn_skirtBack1_mstr_r [{shapeType=Box, extentX=1.01, extentY=1.01, extentZ=1.01}]
 dyn_skirtBack1_r [{shapeType=Box, extentX=15, extentY=10, extentZ=5, center=<7.5 0 0>}]
 dyn_skirtBack2 [{shapeType=Box, extentX=15, extentY=10, extentZ=3, center=<7.5 0 0>}]
 dyn_skirtBack2_l [{shapeType=Box, extentX=15, extentY=10, extentZ=3, center=<7.5 0 0>}]
 dyn_skirtBack2_r [{shapeType=Box, extentX=15, extentY=10, extentZ=3, center=<7.5 0 0>}]
 dyn_skirtBack_mstr [{shapeType=Box, extentX=1.01, extentY=1.01, extentZ=1.01}]
 dyn_skirtFront1_l [{shapeType=Box, extentX=15, extentY=10, extentZ=3, center=<7.5 0 0>}]
 dyn_skirtFront1_mstr_l [{shapeType=Box, extentX=1.01, extentY=1.01, extentZ=1.01}]
 dyn_skirtFront1_mstr_r [{shapeType=Box, extentX=1.01, extentY=1.01, extentZ=1.01}]
 dyn_skirtFront1_r [{shapeType=Box, extentX=15, extentY=10, extentZ=3, center=<7.5 0 0>}]
 dyn_skirtFront2_l [{shapeType=Box, extentX=15, extentY=20, extentZ=3, center=<7.5 0 0>}]
 dyn_skirtFront2_r [{shapeType=Box, extentX=15, extentY=20, extentZ=3, center=<7.5 0 0>}]
 dyn_skirtSide1_l [{shapeType=Box, extentX=15, extentY=10, extentZ=3, center=<7.5 0 0>}]
 dyn_skirtSide1_mstr_l [{shapeType=Box, extentX=1.01, extentY=1.01, extentZ=1.01}]
 dyn_skirtSide1_mstr_r [{shapeType=Box, extentX=1.01, extentY=1.01, extentZ=1.01}]
 dyn_skirtSide1_r [{shapeType=Box, extentX=15, extentY=10, extentZ=3, center=<7.5 0 0>}]
 dyn_skirtSide2_l [{shapeType=Box, extentX=15, extentY=10, extentZ=3, center=<7.5 0 0>}]
 dyn_skirtSide2_r [{shapeType=Box, extentX=15, extentY=21.5, extentZ=3, center=<7.5 -3.9395 0>}]
 thigh_twist_01_l [{shapeType=Capsule, length=23.7626, radius=9.3565, center=<5.2042 0.5973 0.1651>, rotation=<rot 90 0 0>}]
 thigh_twist_01_r [{shapeType=Capsule, length=23.7626, radius=9.3565, center=<-4.845 -0.9865 -0.1651>, rotation=<rot 90 0 0>}]
}
constraints [
 {bone1Name=dyn_skirtFront2_l, bone2Name=dyn_skirtFront1_l, swing1LimitDegrees=2, swing2LimitDegrees=10, twistLimitDegrees=5}
 {bone1Name=dyn_skirtFront2_r, bone2Name=dyn_skirtFront1_r, swing1LimitDegrees=2, swing2LimitDegrees=10, twistLimitDegrees=5}
 {bone1Name=dyn_skirtFront1_l, bone2Name=dyn_skirtFront1_mstr_l, swing1LimitDegrees=2, swing2LimitDegrees=15, twistLimitDegrees=5}
 {bone1Name=dyn_skirtFront1_r, bone2Name=dyn_skirtFront1_mstr_r, swing1LimitDegrees=2, swing2LimitDegrees=15, twistLimitDegrees=5}
 {bone1Name=dyn_skirtSide1_l, bone2Name=dyn_skirtSide1_mstr_l, swing1LimitDegrees=2, swing2LimitDegrees=15, twistLimitDegrees=5}
 {bone1Name=dyn_skirtSide2_l, bone2Name=dyn_skirtSide1_l, swing1LimitDegrees=2, swing2LimitDegrees=10, twistLimitDegrees=5}
 {bone1Name=dyn_skirtBack1_l, bone2Name=dyn_skirtBack1_mstr_l, swing1LimitDegrees=2, swing2LimitDegrees=15, twistLimitDegrees=5}
 {bone1Name=dyn_skirtBack2_l, bone2Name=dyn_skirtBack1_l, swing1LimitDegrees=2, swing2LimitDegrees=10, twistLimitDegrees=5}
 {bone1Name=dyn_skirtBack1, bone2Name=dyn_skirtBack_mstr, swing1LimitDegrees=2, swing2LimitDegrees=15, twistLimitDegrees=5}
 {bone1Name=dyn_skirtBack2, bone2Name=dyn_skirtBack1, swing1LimitDegrees=2, swing2LimitDegrees=10, twistLimitDegrees=5}
 {bone1Name=dyn_skirtBack1_r, bone2Name=dyn_skirtBack1_mstr_r, swing1LimitDegrees=2, swing2LimitDegrees=15, twistLimitDegrees=5}
 {bone1Name=dyn_skirtBack2_r, bone2Name=dyn_skirtBack1_r, swing1LimitDegrees=2, swing2LimitDegrees=10, twistLimitDegrees=5}
 {bone1Name=dyn_skirtSide1_r, bone2Name=dyn_skirtSide1_mstr_r, swing1LimitDegrees=2, swing2LimitDegrees=15, twistLimitDegrees=5}
 {bone1Name=dyn_skirtSide2_r, bone2Name=dyn_skirtSide1_r, swing1LimitDegrees=2, swing2LimitDegrees=10, twistLimitDegrees=5}
]
