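# Physics asset for the Pumpkin Spice female medium upper-body cosmetic, covering its strings, loops and pumpkin charm.
# Capsules on spine_04 and spine_05 form the torso, with two master anchor bodies jointed to spine_05 with free rotation.
# The string master branches into left and right two-segment string and loop chains with tight angular limits.
# The charm hangs from its own master, with twist and one swing axis locked for hinge-like motion.
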PhysicsAsset /BRCosmetics/Characters/Player/Female/Medium/Bodies/F_MED_Pumpkin_Spice/Meshes/F_MED_Pumpkin_Spice_UpperBody_Physics
bodies {
 dyn_loop1_l [{shapeType=Box, extentX=3.3622, extentY=2.379, extentZ=0.5, center=<1.5574 0.2422 0.2584>}]
 dyn_loop1_r [{shapeType=Box, extentX=3.8513, extentY=2.9229, extentZ=0.5, center=<1.6932 0.0071 0.1699>}]
 dyn_loop2_l [{
   shapeType=Capsule
   length=2.75
   radius=0.5
   shapeName=dyn_loop2_l_capsule
   center=<1.2692 -0.0152 0.0653>
   rotation=<rot 0.0003 0.0001 -90.0002>
  }]
 dyn_loop2_r [{shapeType=Capsule, length=2.75, radius=0.5, shapeName=dyn_loop2_r_capsule, center=<1.5 -0.0024 0.2504>, rotation=<rot -90 -17.3231 107.3231>}]
 dyn_pumpkinCharm [
  {shapeType=Capsule, length=3, radius=1.5, shapeName=dyn_pumpkinCharm_capsule, center=<2.5602 0 1.3286>, rotation=<rot 90 0 0>}
  {shapeType=Capsule, length=2, radius=1.5, shapeName=dyn_pumpkinCharm_capsule, center=<2.8238 0 1.2231>, rotation=<rot 0 0 90>}
 ]
 dyn_pumpkinCharm_mstr [{shapeType=Box, extentX=1.01, extentY=1.01, extentZ=1.01}]
 dyn_string1_l [{shapeType=Capsule, length=4, radius=0.5, shapeName=dyn_string1_l_capsule, center=<2 0 0>, rotation=<rot 90 0 0>}]
 dyn_string1_r [{shapeType=Capsule, length=5, radius=0.5, shapeName=dyn_string1_r_capsule, center=<2.25 0 0>, rotation=<rot 90 0 0>}]
 dyn_string2_l [{shapeType=Capsule, length=3.5, radius=0.5, shapeName=dyn_string2_l_capsule, center=<1.65 0 0>, rotation=<rot 90 0 0>}]
 dyn_string2_r [{shapeType=Capsule, length=4.4, radius=0.5, shapeName=dyn_string2_r_capsule, center=<2 0 0>, rotation=<rot 90 0 0>}]
 dyn_string_mstr [{shapeType=Capsule, length=0.505, radius=0.505}]
 spine_04 [{shapeType=Capsule, length=13.0381, radius=7.75, center=<-0.8968 6.5705 0>}]
 spine_05 [{shapeType=Capsule, length=12.374, radius=9, center=<3.167 -2.4335 0.0001>}]
}
constraints [
 {bone1Name=spine_05, bone2Name=spine_04, swing1LimitDegrees=45, swing2LimitDegrees=45, twistLimitDegrees=45}
 {
  bone1Name=dyn_string_mstr
  bone2Name=spine_05
  swing1LimitDegrees=45
  swing1Motion=Free
  swing2LimitDegrees=45
  swing2Motion=Free
  twistLimitDegrees=45
  twistMotion=Free
 }
 {bone1Name=dyn_loop1_l, bone2Name=dyn_string_mstr, swing1LimitDegrees=10, swing2LimitDegrees=10, twistLimitDegrees=5}
 {bone1Name=dyn_loop2_l, bone2Name=dyn_loop1_l, swing1LimitDegrees=15, swing2LimitDegrees=15, twistLimitDegrees=5}
 {bone1Name=dyn_string1_l, bone2Name=dyn_string_mstr, swing1LimitDegrees=10, swing2LimitDegrees=10, twistLimitDegrees=5}
 {bone1Name=dyn_string2_l, bone2Name=dyn_string1_l, swing1LimitDegrees=15, swing2LimitDegrees=15, twistLimitDegrees=5}
 {bone1Name=dyn_string1_r, bone2Name=dyn_string_mstr, swing1LimitDegrees=10, swing2LimitDegrees=10, twistLimitDegrees=5}
 {bone1Name=dyn_string2_r, bone2Name=dyn_string1_r, swing1LimitDegrees=15, swing2LimitDegrees=15, twistLimitDegrees=5}
 {bone1Name=dyn_loop1_r, bone2Name=dyn_string_mstr, swing1LimitDegrees=10, swing2LimitDegrees=10, twistLimitDegrees=5}
 {bone1Name=dyn_loop2_r, bone2Name=dyn_loop1_r, swing1LimitDegrees=15, swing2LimitDegrees=15, twistLimitDegrees=5}
 {
  bone1Name=dyn_pumpkinCharm_mstr
  bone2Name=spine_05
  swing1LimitDegrees=45
  swing1Motion=Free
  swing2LimitDegrees=45
  swing2Motion=Free
  twistLimitDegrees=45
  twistMotion=Free
 }
 {
  bone1Name=dyn_pumpkinCharm
  bone2Name=dyn_pumpkinCharm_mstr
  swing1LimitDegrees=45
  swing1Motion=Locked
  swing2LimitDegrees=45
  twistLimitDegrees=15
  twistMotion=Locked
 }
]
