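# Post-process style AnimBlueprint for the Pumpkin Spice female cosmetic body, animating skirt, pouches, spoon and other accessories over the pose copied from the parent mesh.
# A crouch-aware skirt Control Rig feeds weighted bone constraints, RigidBody simulations, corrective look-at, transform and constraint nodes, then a second Control Rig.
# The event graph derives alphas from crouch state, right thigh angle and left arm angle while targeting, easing physics and correctives; skydiving tilts the pumpkin.
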AnimBlueprint /BRCosmetics/Characters/Player/Female/Medium/Bodies/F_MED_Pumpkin_Spice/Meshes/F_MED_Pumpkin_Spice_AnimBP
target_skeleton=&/F_MED_Pumpkin_Spice_Skeleton
graphs [
 AnimGraph {nodes [
   Look At - Bone: dyn_skirtSide2_r(Alpha=1) {
    alphaBoolBlend {blendOption=Linear}
    boneToModify {boneName=dyn_skirtSide2_r}
    lookAtLocation=<-37.4679 0 38.3081>
    lookAtTarget {boneReference {boneName=pelvis}}
    lookAt_Axis {axis=<1 0>}
   }
   n1 = Control Rig!(Source=Copy Pose From Mesh() {bUseAttachedParent=true}, Alpha, crouching:double="To Float (Boolean)"(InBool=Get bIsCrouching())) -> (Pose) {
    alphaBoolBlend {blendOption=Linear}
    alphaScaleBiasClamp {bInterpResult=true, interpSpeedDecreasing=5, interpSpeedIncreasing=5}
    controlRigAssetReference {blueprintRigClass=/BRCosmetics/Characters/Player/Female/Medium/Bodies/F_MED_Pumpkin_Spice/Meshes/F_MED_Pumpkin_Spice_CtrlRig/F_MED_Pumpkin_Spice_Skirt_CR.F_MED_Pumpkin_Spice_Skirt_CR_C}
   }
   n2 = "Constraint - dyn_skirtFront1 (PP)"!(ComponentPose=Local To Component(LocalPose=@n1), Alpha=1, ConstraintWeights_0=0.5, ConstraintWeights_1:float=0.5) -> (Pose) {
    alphaBoolBlend {blendOption=Linear}
    boneToModify {boneName=dyn_skirtFront1}
    constraintSetup [{targetBone {boneName=dyn_skirtFront1_l}}, {targetBone {boneName=dyn_skirtFront1_r}}]
    constraintWeights [1,1]
   }
   n3 = "Constraint - dyn_skirtFront2 (PP)"!(ComponentPose=@n2, Alpha=1, ConstraintWeights_0=0.5, ConstraintWeights_1:float=0.5) -> (Pose) {
    alphaBoolBlend {blendOption=Linear}
    boneToModify {boneName=dyn_skirtFront2}
    constraintSetup [{targetBone {boneName=dyn_skirtFront2_l}}, {targetBone {boneName=dyn_skirtFront2_r}}]
    constraintWeights [1,1]
   }
   n4 = "Constraint - dyn_skirtBack1 (PP)"!(ComponentPose=@n3, Alpha=1, ConstraintWeights_0=0.5, ConstraintWeights_1:float=0.5) -> (Pose) {
    alphaBoolBlend {blendOption=Linear}
    boneToModify {boneName=dyn_skirtBack1}
    constraintSetup [{targetBone {boneName=dyn_skirtBack1_l}}, {targetBone {boneName=dyn_skirtBack1_r}}]
    constraintWeights [1,1]
   }
   n5 = "Constraint - dyn_skirtBack2 (PP)"!(ComponentPose=@n4, Alpha=1, ConstraintWeights_0=0.5, ConstraintWeights_1:float=0.5) -> (Pose) {
    alphaBoolBlend {blendOption=Linear}
    boneToModify {boneName=dyn_skirtBack2}
    constraintSetup [{targetBone {boneName=dyn_skirtBack2_l}}, {targetBone {boneName=dyn_skirtBack2_r}}]
    constraintWeights [1,1]
   }
   n6 = "Constraint - dyn_scraper_mstr (PPPPPP)"!(
    ComponentPose=@n5
    Alpha=1
    ConstraintWeights_0=0
    ConstraintWeights_1:float=0.2117
    ConstraintWeights_2:float=0.3899
    ConstraintWeights_3:float=0.0664
    ConstraintWeights_4:float=0.1591
    ConstraintWeights_5:float=0.1729
   ) -> (Pose) {
    alphaBoolBlend {blendOption=Linear}
    boneToModify {boneName=dyn_scraper_mstr}
    constraintSetup [
     {targetBone {boneName=upperarm_twist_02_r}}
     {targetBone {boneName=spine_01}}
     {targetBone {boneName=spine_02}}
     {targetBone {boneName=spine_03}}
     {targetBone {boneName=spine_04}}
     {targetBone {boneName=pelvis}}
    ]
    constraintWeights [1,1,1,1,1,1]
   }
   n7 = "Constraint - dyn_pumpkinCharm_mstr (PPPPPPP)"!(
    ComponentPose=@n6
    Alpha=1
    ConstraintWeights_0=0.1211
    ConstraintWeights_1:float=0.0023
    ConstraintWeights_2:float=0.3713
    ConstraintWeights_3:float=0.3067
    ConstraintWeights_4:float=0.003
    ConstraintWeights_5:float=0.0148
    ConstraintWeights_6:float=0.1809
   ) -> (Pose) {
    alphaBoolBlend {blendOption=Linear}
    boneToModify {boneName=dyn_pumpkinCharm_mstr}
    constraintSetup [
     {targetBone {boneName=head}}
     {targetBone {boneName=clavicle_l}}
     {targetBone {boneName=spine_05}}
     {targetBone {boneName=neck_02}}
     {targetBone {boneName=clavicle_r}}
     {targetBone {boneName=C_jaw}}
     {targetBone {boneName=neck_01}}
    ]
    constraintWeights [1,1,1,1,1,1,1]
   }
   n8 = "Constraint - dyn_string_mstr (PP)"!(ComponentPose=@n7, Alpha=1, ConstraintWeights_0=0.607, ConstraintWeights_1:float=0.393) -> (Pose) {
    alphaBoolBlend {blendOption=Linear}
    boneToModify {boneName=dyn_string_mstr}
    constraintSetup [{targetBone {boneName=spine_04}}, {targetBone {boneName=spine_05}}]
    constraintWeights [1,1]
   }
   n9 = "Constraint - dyn_spoonLoop (PPPPP)"!(
    ComponentPose=@n8
    Alpha=1
    ConstraintWeights_0=0.2574
    ConstraintWeights_1:float=0.4346
    ConstraintWeights_2:float=0.0419
    ConstraintWeights_3:float=0.1319
    ConstraintWeights_4:float=0.1342
   ) -> (Pose) {
    alphaBoolBlend {blendOption=Linear}
    boneToModify {boneName=dyn_spoonLoop}
    constraintSetup [
     {targetBone {boneName=spine_01}}
     {targetBone {boneName=spine_02}}
     {targetBone {boneName=spine_03}}
     {targetBone {boneName=spine_04}}
     {targetBone {boneName=pelvis}}
    ]
    constraintWeights [1,1,1,1,1]
   }
   n10 = RigidBody(ComponentPose=@n9) {
    bClampLinearTranslationLimitToRefPose=true
    bOverrideWorldGravity=true
    overridePhysicsAsset=&/F_MED_Pumpkin_Spice_Accessories_Physics
    simulationSpace=BaseBoneSpace
    alphaBoolBlend {blendOption=Linear}
    baseBoneRef {boneName=spine_01}
    componentAppliedLinearAccClamp=<5000 5000 5000>
    componentLinearAccScale=<0.4 0.4 0.4>
    componentLinearVelScale=<0.3 0.3 0.3>
    overrideWorldGravity {z=-980}
   }
   n11 = Get isNotCrouching() -> (isNotCrouching:double)
   n12 = RigidBody(ComponentPose=@n10, Alpha=@n11) {
    bClampLinearTranslationLimitToRefPose=true
    bOverrideWorldGravity=true
    overridePhysicsAsset=&/F_MED_Pumpkin_Spice_Physics_Spoon
    simulationSpace=BaseBoneSpace
    alphaBoolBlend {blendOption=Linear}
    alphaScaleBiasClamp {bInterpResult=true, interpSpeedDecreasing=7, interpSpeedIncreasing=7}
    baseBoneRef {boneName=pelvis}
    componentAppliedLinearAccClamp=<5000 5000 5000>
    componentLinearAccScale=<0.4 0.4 0.4>
    componentLinearVelScale=<0.3 0.3 0.3>
    overrideWorldGravity {z=-980}
   }
   n13 = RigidBody(ComponentPose=@n12) {
    bClampLinearTranslationLimitToRefPose=true
    bOverrideWorldGravity=true
    overridePhysicsAsset=&/F_MED_Pumpkin_Spice_Pouch_Physics
    simulationSpace=BaseBoneSpace
    alphaBoolBlend {blendOption=Linear}
    baseBoneRef {boneName=spine_01}
    componentAppliedLinearAccClamp=<5000 5000 5000>
    componentLinearAccScale=<0.4 0.4 0.4>
    componentLinearVelScale=<0.3 0.3 0.3>
    overrideWorldGravity {z=-980}
   }
   n14 = RigidBody(ComponentPose=@n13) {
    bClampLinearTranslationLimitToRefPose=true
    bOverrideWorldGravity=true
    overridePhysicsAsset=&/F_MED_Pumpkin_Spice_UpperBody_Physics
    simulationSpace=BaseBoneSpace
    alphaBoolBlend {blendOption=Linear}
    baseBoneRef {boneName=spine_03}
    componentAppliedLinearAccClamp=<5000 5000 5000>
    componentLinearAccScale=<0.4 0.4 0.4>
    componentLinearVelScale=<0.3 0.3 0.3>
    overrideWorldGravity {z=-980}
   }
   n15 = RigidBody(ComponentPose=@n14, Alpha=@n11) {
    bClampLinearTranslationLimitToRefPose=true
    bOverrideWorldGravity=true
    overridePhysicsAsset=&/F_MED_Pumpkin_Spice_Skirt_Physics
    simulationSpace=BaseBoneSpace
    alphaBoolBlend {blendOption=Linear}
    alphaScaleBiasClamp {bInterpResult=true, interpSpeedDecreasing=5, interpSpeedIncreasing=5}
    baseBoneRef {boneName=pelvis}
    componentAppliedLinearAccClamp=<5000 5000 5000>
    componentLinearAccScale=<0.4 0.4 0.4>
    componentLinearVelScale=<0.3 0.3 0.3>
    overrideWorldGravity {z=-980}
   }
   n16 = RigidBody!(ComponentPose=@n15, bAlphaBoolEnabled:bool, ExternalForce) -> (Pose) {
    alphaInputType=Bool
    bClampLinearTranslationLimitToRefPose=true
    bOverrideWorldGravity=true
    overridePhysicsAsset=&/F_MED_Pumpkin_Spice_SkirtCrouch_Physics
    simulationSpace=BaseBoneSpace
    alphaBoolBlend {blendOption=Linear}
    alphaScaleBiasClamp {bInterpResult=true, interpSpeedDecreasing=5, interpSpeedIncreasing=5}
    baseBoneRef {boneName=pelvis}
    componentAppliedLinearAccClamp=<5000 5000 5000>
    componentLinearAccScale=<0.4 0.4 0.4>
    componentLinearVelScale=<0.3 0.3 0.3>
    overrideWorldGravity {z=-980}
   }
   n17 = Get Right Leg() -> (Right Leg:double)
   n18 = Look At - Bone: dyn_skirtSide1_r(ComponentPose=@n16, Alpha=@n17) {
    alphaBoolBlend {blendOption=Linear}
    alphaScaleBiasClamp {bClampResult=true, bInterpResult=true, clampMax=0.5}
    boneToModify {boneName=dyn_skirtSide1_r}
    lookAtLocation=<2.4819 0.4885 8.8158>
    lookAtTarget {boneReference {boneName=calf_r}}
    lookAt_Axis {axis=<1 0>}
   }
   n19 = "Transform (Modify) Bone - Bone: dyn_skirtSide1_r"(ComponentPose=@n18, Alpha=@n17, Translation="0,0,0", Rotation="0,0,28", Scale="1,1,1") {
    rotationMode=BMM_Additive
    rotationSpace=BCS_BoneSpace
    translationMode=BMM_Additive
    translationSpace=BCS_BoneSpace
    alphaBoolBlend {blendOption=Linear}
    alphaScaleBiasClamp {bInterpResult=true}
    boneToModify {boneName=dyn_skirtSide1_r}
    rotation {roll=28}
   }
   n20 = "Transform (Modify) Bone - Bone: dyn_spoonLoop"(ComponentPose=@n19, Alpha=1, Translation="0,-2.5,2", Rotation="0,-15,0", Scale="1,1,1") {
    rotationMode=BMM_Additive
    translationMode=BMM_Additive
    alphaBoolBlend {blendOption=Linear}
    alphaScaleBiasClamp {bInterpResult=true, interpSpeedDecreasing=7, interpSpeedIncreasing=7}
    boneToModify {boneName=dyn_spoonLoop}
    rotation {yaw=-15}
    translation {y=-2.5, z=2}
   }
   n21 = "Transform (Modify) Bone - Bone: dyn_pouch_mstr_l"(ComponentPose=@n20, Alpha=1, Translation="0,-3,2", Rotation="0,-40,20", Scale="1,1,1") {
    rotationMode=BMM_Additive
    translationMode=BMM_Additive
    alphaBoolBlend {blendOption=Linear}
    alphaScaleBiasClamp {bInterpResult=true, interpSpeedDecreasing=7, interpSpeedIncreasing=7}
    boneToModify {boneName=dyn_pouch_mstr_l}
    rotation {roll=20, yaw=-40}
    translation {y=-3, z=2}
   }
   n22 = "Transform (Modify) Bone - Bone: dyn_pouch_mstr_r"(ComponentPose=@n21, Alpha=1, Translation="-0.574,3.486,-0.885", Rotation="-15.045,23.891,38.675", Scale="1,1,1") {
    rotationMode=BMM_Additive
    rotationSpace=BCS_BoneSpace
    translationMode=BMM_Additive
    translationSpace=BCS_BoneSpace
    alphaBoolBlend {blendOption=Linear}
    alphaScaleBiasClamp {bInterpResult=true, interpSpeedDecreasing=7, interpSpeedIncreasing=7}
    boneToModify {boneName=dyn_pouch_mstr_r}
    rotation=<rot -15.045 23.891 38.675>
    translation=<-0.574 3.486 -0.885>
   }
   n23 = "Transform (Modify) Bone - Bone: dyn_skirtFront3_l"(ComponentPose=@n22, Alpha=1, Translation="0,2,-2", Rotation="-35,0,0", Scale="1,1,1") {
    rotationMode=BMM_Additive
    rotationSpace=BCS_BoneSpace
    translationMode=BMM_Additive
    translationSpace=BCS_BoneSpace
    alphaBoolBlend {blendOption=Linear}
    alphaScaleBiasClamp {bInterpResult=true, interpSpeedDecreasing=7, interpSpeedIncreasing=7}
    boneToModify {boneName=dyn_skirtFront3_l}
    rotation {pitch=-35}
    translation {y=2, z=-2}
   }
   n24 = "Transform (Modify) Bone - Bone: dyn_skirtFront2_r"(ComponentPose=@n23, Alpha=1, Translation="-0.000,0.4,0.009", Rotation="10,10,0", Scale="1,1,1") {
    rotationMode=BMM_Additive
    rotationSpace=BCS_BoneSpace
    translationMode=BMM_Additive
    translationSpace=BCS_BoneSpace
    alphaBoolBlend {blendOption=Linear}
    alphaScaleBiasClamp {bInterpResult=true, interpSpeedDecreasing=7, interpSpeedIncreasing=7}
    boneToModify {boneName=dyn_skirtFront2_r}
    rotation {pitch=10, yaw=10}
    translation {y=0.4, z=0.009}
   }
   n25 = "Transform (Modify) Bone - Bone: dyn_skirtSide1_r"(ComponentPose=@n24, Alpha=1, Translation="0,0,0", Rotation="15,0,0.000", Scale="1,1,1") {
    rotationMode=BMM_Additive
    rotationSpace=BCS_BoneSpace
    translationMode=BMM_Additive
    translationSpace=BCS_BoneSpace
    alphaBoolBlend {blendOption=Linear}
    alphaScaleBiasClamp {bInterpResult=true, interpSpeedDecreasing=7, interpSpeedIncreasing=7}
    boneToModify {boneName=dyn_skirtSide1_r}
    rotation {pitch=15}
   }
   n26 = Get isNotCrouching()
   n27 = "Constraint - dyn_skirtFront2_r (P)"(ComponentPose=@n25, Alpha=@n26, ConstraintWeights_0=1) {
    alphaBoolBlend {blendOption=Linear}
    alphaScaleBiasClamp {bClampResult=true, bInterpResult=true, clampMax=0.4, interpSpeedDecreasing=5, interpSpeedIncreasing=7}
    boneToModify {boneName=dyn_skirtFront2_r}
    constraintSetup [{targetBone {boneName=thigh_r}}]
    constraintWeights [1]
   }
   n28 = "Constraint - dyn_skirtSide2_r (P)"(ComponentPose=@n27, Alpha=@n26, ConstraintWeights_0=1) {
    alphaBoolBlend {blendOption=Linear}
    alphaScaleBiasClamp {bClampResult=true, bInterpResult=true, clampMax=0.4, interpSpeedDecreasing=5, interpSpeedIncreasing=7}
    boneToModify {boneName=dyn_skirtSide2_r}
    constraintSetup [{targetBone {boneName=thigh_r}}]
    constraintWeights [1]
   }
   n29 = Get arm_l() -> (arm_l:double)
   n30 = "Constraint - dyn_spoon (PP)"!(ComponentPose=@n28, Alpha=@n29, ConstraintWeights_0=1, ConstraintWeights_1:float=0) -> (Pose) {
    alphaBoolBlend {blendOption=Linear}
    alphaScaleBiasClamp {bClampResult=true, bInterpResult=true, clampMax=0.65, interpSpeedDecreasing=5, interpSpeedIncreasing=7}
    boneToModify {boneName=dyn_spoon}
    constraintSetup [{targetBone {boneName=dyn_pouch_l}}, {targetBone {boneName=dyn_pouch_mstr_l}}]
    constraintWeights [1,1]
   }
   n31 = "Transform (Modify) Bone - Bone: dyn_spoonLoop"(ComponentPose=@n30, Alpha=@n29, Translation="0,-5,0", Rotation="0,-15,0.000", Scale="1,1,1") {
    rotationMode=BMM_Additive
    rotationSpace=BCS_BoneSpace
    translationMode=BMM_Additive
    translationSpace=BCS_BoneSpace
    alphaBoolBlend {blendOption=Linear}
    alphaScaleBiasClamp {bInterpResult=true, interpSpeedDecreasing=7, interpSpeedIncreasing=7}
    boneToModify {boneName=dyn_spoonLoop}
    rotation {yaw=-15}
    translation {y=-5}
   }
   n32 = "Transform (Modify) Bone - Bone: dyn_pouch_mstr_l"(ComponentPose=@n31, Alpha=@n29, Translation="0,-1,0", Rotation="0,-50,0.000", Scale="1,1,1") {
    rotationMode=BMM_Additive
    rotationSpace=BCS_BoneSpace
    translationMode=BMM_Additive
    translationSpace=BCS_BoneSpace
    alphaBoolBlend {blendOption=Linear}
    alphaScaleBiasClamp {bInterpResult=true, interpSpeedDecreasing=7, interpSpeedIncreasing=7}
    boneToModify {boneName=dyn_pouch_mstr_l}
    rotation {yaw=-50}
    translation {y=-1}
   }
   n33 = "Transform (Modify) Bone - Bone: dyn_pumpkin_mstr"!(ComponentPose=@n32, bAlphaBoolEnabled:bool=Get bIsSkyDiving(), Translation="0,0,0", Rotation="20,0.000,0.000", Scale="1,1,1") -> (Pose) {
    alphaInputType=Bool
    rotationMode=BMM_Additive
    rotationSpace=BCS_BoneSpace
    translationMode=BMM_Additive
    alphaBoolBlend {blendOption=Linear}
    alphaScaleBiasClamp {bInterpResult=true, interpSpeedDecreasing=7, interpSpeedIncreasing=7}
    boneToModify {boneName=dyn_pumpkin_mstr}
    rotation {pitch=20}
   }
   n34 = Control Rig!(Source=Component To Local(ComponentPose=@n33), Alpha) -> (Pose) {
    alphaBoolBlend {blendOption=Linear}
    controlRigAssetReference {blueprintRigClass=/BRCosmetics/Characters/Player/Female/Medium/Bodies/F_MED_Pumpkin_Spice/Meshes/F_MED_Pumpkin_Spice_CtrlRig/F_MED_Pumpkin_Spice_CtrlRig.F_MED_Pumpkin_Spice_CtrlRig_C}
   }
   n35 = "Transform (Modify) Bone - Bone: dyn_pouch_mstr_l"(ComponentPose=Local To Component(LocalPose=@n34), Alpha=1, Translation="-0.110,-1.791,-0.572", Rotation="0.000,0.000,-20.000", Scale="1,1,1") {
    rotationMode=BMM_Additive
    rotationSpace=BCS_BoneSpace
    translationMode=BMM_Additive
    translationSpace=BCS_BoneSpace
    alphaBoolBlend {blendOption=Linear}
    alphaScaleBiasClamp {bInterpResult=true, interpSpeedDecreasing=7, interpSpeedIncreasing=7}
    boneToModify {boneName=dyn_pouch_mstr_l}
    rotation {roll=-20}
    translation=<-0.11 -1.791 -0.572>
   }
   Output Pose(Result=Component To Local(ComponentPose=@n35)) {layerGroup=None}
  ]}
 EventGraph {nodes [
   Event BlueprintUpdateAnimation()
   n1 = Get bIsCrouching()
   > Set isNotCrouching(isNotCrouching:double="To Float (Boolean)"(InBool=NOT Boolean(A=@n1))) -> (Output_Get:double)
   n2 = GetOwnerSkeletalMeshComponent()
   n3 = BreakRotator(InRot=BreakTransform(InTransform=GetDeltaTransformFromRefPose(self=@n2, BoneName=upperarm_l, BaseName=spine_03)).Rotation)
   > Set arm_l(arm_l:double=SelectFloat(A=MapRangeClamped(Value=@n3.Yaw, InRangeA=8, InRangeB=-6), B=0, bPickA=Get bIsTargeting())) -> (Output_Get:double)
   n4 = GetOwnerSkeletalMeshComponent()
   > Set Right Leg(Right Leg:double=float * float(
     A:bool=Get bIsCrouching()
     B=GetAlphaFromReferenceTransformDelta(SkeletalMeshComponent=@n4, BoneName=thigh_r, BaseName=pelvis, InputRange=(X=-19,Y=-100), TwistAxis="1,0,0")
    ) {is=title K2Node_PromotableOperator}) -> (Output_Get:double)
  ]}
]
variables [
 arm_l {type=number}
 isNotCrouching {type=number}
 Right Leg {type=number}
]
