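# Control Rig for the Pumpkin Spice female medium character body that procedurally animates the vine bone chain attached to the head.
# A single Chain Harmonics node runs from Begin Execution on the chain rooted at dyn_vine1, the first of eight vine bones.
# Only the wave component is enabled, with noise and an editable curve; pendulum and reach are off.
# Debug drawing is left on, and the imported skeleton also carries spine and face bones the graph does not drive.
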ControlRigBlueprint /BRCosmetics/Characters/Player/Female/Medium/Bodies/F_MED_Pumpkin_Spice/Meshes/Parts/F_MED_Pumpkin_Spice_CtrlRig/F_MED_Pumpkin_Spice_CtrlRig
graphs [
 RigVMFunctionLibrary {links [], nodes []}
 RigVMModel {nodes [
   RigUnit_BeginExecution = &RigUnit_BeginExecution()
   /Script/ControlRig.RigUnit_ChainHarmonics(
    ChainRoot=dyn_vine1
    DrawWorldOffset=(Rotation=(X=0,Y=0,Z=0,W=1),Translation=(X=0,Y=0,Z=0),Scale3D=(X=1,Y=1,Z=1))
    Pendulum=(bEnabled=false,PendulumStiffness=2,PendulumGravity=(X=0,Y=0,Z=0),PendulumBlend=0.75,PendulumDrag=0.98,PendulumMinimum=0,PendulumMaximum=0.1,PendulumEase=Linear,UnwindAxis=(X=0,Y=1,Z=0),UnwindMinimum=0.2,UnwindMaximum=0.05)
    Reach=(bEnabled=false,ReachTarget=(X=10,Y=0,Z=10),ReachAxis=(X=1,Y=0,Z=0),ReachMinimum=0,ReachMaximum=-5,ReachEase=Linear)
    Speed=(X=0.75,Y=0.75,Z=0.75)
    Wave=(bEnabled=True,WaveFrequency=(X=18,Y=18,Z=-18),WaveAmplitude=(X=0.5,Y=0.5,Z=0.5),WaveOffset=(X=0.5,Y=0.5,Z=0.5),WaveNoise=(X=0,Y=0.5,Z=0.5),WaveMinimum=0,WaveMaximum=1,WaveEase=Linear)
    WaveCurve=(EditorCurveData=(Keys=((Value=0.25),(Time=1.0107,Value=0.6084)),DefaultValue=340282346638528859811704183484516925440,PreInfinityExtrap=RCCE_Constant,PostInfinityExtrap=RCCE_Constant))
    bDrawDebug=True
    ExecutePin=@RigUnit_BeginExecution.ExecutePin
   ) {is=struct RigVMUnitNode}
  ]}
]
hierarchy [
 root {loc=<0 0 0>, rot=<0 0 0 1>}
 pelvis {parent=root, loc=<0 -0.5301 98.0571>, rot=<0.0127 0.707 -0.0127 -0.707>}
 spine_01 {parent=pelvis, loc=<6.1208 -0.9328 0>, rot=<0 0 0.034 -0.9994>}
 spine_02 {parent=spine_01, loc=<6.5996 -0.2095 0>, rot=<0 0 0.0238 -0.9997>}
 spine_03 {parent=spine_02, loc=<6.4914 -0.1237 0>, rot=<0 0 0.0149 -0.9999>}
 spine_04 {parent=spine_03, loc=<6.9598 0.0794 0>, rot=<0 0 -0.0132 -0.9999>}
 spine_05 {parent=spine_04, loc=<6.9987 0 0>, rot=<0 0 -0.0015 -1>}
 clavicle_l {parent=spine_05, loc=<5.4831 -0.1053 -3.2946>, rot=<0.0465 -0.7289 -0.009 -0.6829>}
 clavicle_r {parent=spine_05, loc=<5.4831 -0.1053 3.2946>, rot=<0.7289 0.0465 -0.6829 0.009>}
 neck_01 {parent=spine_05, loc=<11.3138 0.1529 0>, rot=<0 0 -0.2003 -0.9797>}
 neck_02 {parent=neck_01, loc=<4.5627 0 0>, rot=<0 0 0.0504 -0.9987>}
 head {parent=neck_02, loc=<5.3914 0 0>, rot=<0 0 0.115 -0.9934>}
 faceAttach {parent=head, loc=<-3.337 -0.4556 0.0011>, rot=<0.0082 -0.7071 -0.0082 -0.7071>, scale=<0.89 0.89 0.89>}
 C_brow_mid {parent=faceAttach, loc=<-0.0222 10.1082 8.3365>, rot=<0 -0.0311 0.0061 -0.9995>}
 L_brow_mid {parent=faceAttach, loc=<3.5589 9.3601 9.6601>, rot=<0 -0.0035 0.2553 -0.9669>}
 L_brow_outer {parent=faceAttach, loc=<5.6605 7.5267 9.7309>, rot=<0.2649 0.3299 0.3373 -0.841>}
 L_cheek_inner {parent=faceAttach, loc=<4.3217 8.3452 3.077>, rot=<-0.0235 -0.1502 0.3322 -0.9309>}
 C_lip_upper_mid {parent=faceAttach, loc=<0.0011 10.0967 0.9809>, rot=<0.2225 0.0019 -0.0007 -0.9749>}
 L_lip_upper_outer {parent=faceAttach, loc=<1.5473 9.1328 0.9604>, rot=<0.0585 -0.0087 0.238 -0.9695>}
 L_lip_corner {parent=faceAttach, loc=<2.3011 8.8532 0.9682>, rot=<0.0022 0.0088 -0.0145 -0.9999>}
 C_nose_bridge {parent=faceAttach, loc=<0.0187 11.2574 5.0563>, rot=<-0.2218 -0.0189 -0.0053 -0.9749>}
 R_brow_mid {parent=faceAttach, loc=<-3.5549 9.3616 9.6658>, rot=<-0.0001 0.0042 -0.2553 -0.9669>}
 R_brow_outer {parent=faceAttach, loc=<-5.6681 7.5102 9.718>, rot=<0.0553 -0.0657 -0.4281 -0.8997>}
 R_cheek_inner {parent=faceAttach, loc=<-4.304 8.3562 3.0688>, rot=<-0.0223 0.147 -0.3325 -0.9313>}
 R_lip_corner {parent=faceAttach, loc=<-2.3435 8.7578 0.9683>, rot=<0.0027 -0.0082 0.0146 -0.9999>}
 C_jaw {parent=faceAttach, loc=<-0.0138 3.0827 1.1287>, rot=<0 0 0 1>}
 L_lip_lower_outer {parent=C_jaw, loc=<1.6361 6.3261 -0.2247>, rot=<0.0109 0.0332 0.2971 -0.9542>}
 C_lip_lower_mid {parent=C_jaw, loc=<0.0224 6.9268 -0.2693>, rot=<0 0.0184 0.0039 -0.9998>}
 R_lip_lower_outer {parent=C_jaw, loc=<-1.6459 6.4171 -0.3218>, rot=<-0.0007 0.0294 -0.2485 -0.9682>}
 R_eye {parent=faceAttach, loc=<-2.9797 6.8618 7.5334>, rot=<0 0 0.9998 -0.0187>}
 R_eye_lid_upper_mid {parent=R_eye, loc=<0.3197 -2.2488 0.896>, rot=<0.0523 -0.0009 0.0167 -0.9985>}
 R_eye_lid_lower_mid {parent=R_eye, loc=<0.2502 -1.8131 -1.0445>, rot=<0.0329 -0.0005 0.0167 -0.9993>}
 L_eye {parent=faceAttach, loc=<3.1803 6.8734 7.5721>, rot=<0 0 0.9998 0.0187>}
 L_eye_lid_upper_mid {parent=L_eye, loc=<-0.1356 -2.2209 0.8332>, rot=<0.0523 0 0 -0.9986>}
 L_eye_lid_lower_mid {parent=L_eye, loc=<-0.044 -1.7806 -1.111>, rot=<0.0329 0 0 -0.9995>}
 R_lip_upper_outer {parent=faceAttach, loc=<-1.5741 9.176 0.9924>, rot=<0.0569 0.0133 -0.2398 -0.9691>}
 dyn_vine1 {parent=head, loc=<23.2683 -4.3376 -7.4947>, rot=<0.3952 -0.3887 0.8097 -0.1929>}
 dyn_vine2 {parent=dyn_vine1, loc=<8.032 0 0>, rot=<-0.5404 -0.019 0.7976 0.2673>}
 dyn_vine3 {parent=dyn_vine2, loc=<6.9716 0 0>, rot=<-0.3885 -0.028 0.6126 -0.6878>}
 dyn_vine4 {parent=dyn_vine3, loc=<13.5047 0 0>, rot=<0.7144 -0.0677 -0.6373 -0.2808>}
 dyn_vine5 {parent=dyn_vine4, loc=<12.9222 0 0>, rot=<0.7373 -0.0583 -0.6054 0.2941>}
 dyn_vine6 {parent=dyn_vine5, loc=<7.9424 0 0>, rot=<-0.4352 -0.0806 0.7603 0.4755>}
 dyn_vine7 {parent=dyn_vine6, loc=<6.9369 0 0>, rot=<-0.5319 -0.0261 0.8378 -0.1202>}
 dyn_vine8 {parent=dyn_vine7, loc=<7.9739 0 0>, rot=<0 0 0 -1>}
]
