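# Physics asset for the hat-off hair mesh of the female Pumpkin Spice player character, providing secondary hair motion around the head.
# Small box bodies at five master bones anchor capsule chains to a head capsule: single-segment front left and right, two-segment side left and right, and two-segment back left only.
# Master-to-head joints leave swing and twist free, while joints along each chain have narrow swing and twist limits.
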PhysicsAsset /BRCosmetics/Characters/Player/Female/Medium/Bodies/F_MED_Pumpkin_Spice/Meshes/Parts/F_MED_Pumpkin_Spice_HairHatOff_Physics
bodies {
 dyn_hairBackHigh_l [{shapeType=Capsule, length=10, radius=5, shapeName=dyn_hairBackHigh_l_capsule, center=<4 0 0>, rotation=<rot -0.0002 0 90>}]
 dyn_hairBack_l [{shapeType=Capsule, length=13, radius=4.75, shapeName=dyn_hairBack_l_capsule, center=<5.5098 -0.1241 0.6525>, rotation=<rot -0.0004 0 90>}]
 dyn_hairBack_mstr_l [{shapeType=Box, extentX=1.01, extentY=1.01, extentZ=1.01}]
 dyn_hairFront_l [{shapeType=Capsule, length=6, radius=3, shapeName=dyn_hairFront_l_capsule, center=<4.5789 1.3665 -0.5678>, rotation=<rot 90 0 0>}]
 dyn_hairFront_mstr_l [{shapeType=Box, extentX=1.01, extentY=1.01, extentZ=1.01}]
 dyn_hairFront_mstr_r [{shapeType=Box, extentX=1.01, extentY=1.01, extentZ=1.01}]
 dyn_hairFront_r [{shapeType=Capsule, length=5.5, radius=3, shapeName=dyn_hairFront_r_capsule, center=<5.0152 -1.2101 -0.0125>, rotation=<rot 0 0 90>}]
 dyn_hairSideHigh_l [{shapeType=Capsule, length=8, radius=4, shapeName=dyn_hairSideHigh_l_capsule, center=<4 0 0>, rotation=<rot 0 0 90>}]
 dyn_hairSideHigh_r [{shapeType=Capsule, length=8, radius=4, shapeName=dyn_hairSideHigh_r_capsule, center=<4 0 0>, rotation=<rot 0 0 90>}]
 dyn_hairSide_l [{shapeType=Capsule, length=8.5, radius=3.5, shapeName=dyn_hairSide_l_capsule, center=<5.9382 0.4887 1>, rotation=<rot 0 0.0004 90>}]
 dyn_hairSide_mstr_l [{shapeType=Box, extentX=1.01, extentY=1.01, extentZ=1.01}]
 dyn_hairSide_mstr_r [{shapeType=Box, extentX=1.01, extentY=1.01, extentZ=1.01}]
 dyn_hairSide_r [{shapeType=Capsule, length=5, radius=4.25, shapeName=dyn_hairSide_r_capsule, center=<7.1855 0.157 0.3911>, rotation=<rot 0 -0.0001 90>}]
 head [{shapeType=Capsule, length=10, radius=6.25, shapeName=head_capsule, center=<0 1.2993 0>, rotation=<rot 90 0 0>}]
}
constraints [
 {bone1Name=dyn_hairFront_l, bone2Name=dyn_hairFront_mstr_l, swing1LimitDegrees=15, swing2LimitDegrees=15, twistLimitDegrees=5}
 {bone1Name=dyn_hairFront_r, bone2Name=dyn_hairFront_mstr_r, swing1LimitDegrees=15, swing2LimitDegrees=15, twistLimitDegrees=5}
 {bone1Name=dyn_hairBack_l, bone2Name=dyn_hairBackHigh_l, swing1LimitDegrees=15, swing2LimitDegrees=10, twistLimitDegrees=5}
 {bone1Name=dyn_hairBackHigh_l, bone2Name=dyn_hairBack_mstr_l, swing1LimitDegrees=20, swing2LimitDegrees=15, twistLimitDegrees=5}
 {bone1Name=dyn_hairSide_l, bone2Name=dyn_hairSideHigh_l, swing1LimitDegrees=10, swing2LimitDegrees=15, twistLimitDegrees=5}
 {bone1Name=dyn_hairSideHigh_l, bone2Name=dyn_hairSide_mstr_l, swing1LimitDegrees=15, swing2LimitDegrees=20, twistLimitDegrees=5}
 {bone1Name=dyn_hairSide_r, bone2Name=dyn_hairSideHigh_r, swing1LimitDegrees=10, swing2LimitDegrees=15, twistLimitDegrees=5}
 {bone1Name=dyn_hairSideHigh_r, bone2Name=dyn_hairSide_mstr_r, swing1LimitDegrees=15, swing2LimitDegrees=20, twistLimitDegrees=5}
 {
  bone1Name=dyn_hairFront_mstr_l
  bone2Name=head
  swing1LimitDegrees=45
  swing1Motion=Free
  swing2LimitDegrees=45
  swing2Motion=Free
  twistLimitDegrees=45
  twistMotion=Free
 }
 {
  bone1Name=dyn_hairFront_mstr_r
  bone2Name=head
  swing1LimitDegrees=45
  swing1Motion=Free
  swing2LimitDegrees=45
  swing2Motion=Free
  twistLimitDegrees=45
  twistMotion=Free
 }
 {
  bone1Name=dyn_hairSide_mstr_r
  bone2Name=head
  swing1LimitDegrees=45
  swing1Motion=Free
  swing2LimitDegrees=45
  swing2Motion=Free
  twistLimitDegrees=45
  twistMotion=Free
 }
 {
  bone1Name=dyn_hairSide_mstr_l
  bone2Name=head
  swing1LimitDegrees=45
  swing1Motion=Free
  swing2LimitDegrees=45
  swing2Motion=Free
  twistLimitDegrees=45
  twistMotion=Free
 }
 {
  bone1Name=dyn_hairBack_mstr_l
  bone2Name=head
  swing1LimitDegrees=45
  swing1Motion=Free
  swing2LimitDegrees=45
  swing2Motion=Free
  twistLimitDegrees=45
  twistMotion=Free
 }
]
