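# Physics asset for the hair part of the F_MED_Pumpkin_Spice female player cosmetic, defining bodies and joints for five dynamic hair chains.
# Capsules cover the front and side locks on both sides plus one left back lock, each chain starting at a small box on a master bone.
# Front locks join their masters directly with tight swing and twist angles.
# Side and back locks use slightly wider swing through tiny "High" capsules, which rotate freely on their masters.
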PhysicsAsset /BRCosmetics/Characters/Player/Female/Medium/Bodies/F_MED_Pumpkin_Spice/Meshes/Parts/F_MED_Pumpkin_Spice_Hair_Physics
bodies {
 dyn_hairBackHigh_l [{shapeType=Capsule, length=0.505, radius=0.505}]
 dyn_hairBack_l [{shapeType=Capsule, length=8.75, radius=4.25, shapeName=dyn_hairBack_l_capsule, center=<5.787 0 0>, rotation=<rot 0 0 90>}]
 dyn_hairBack_mstr_l [{shapeType=Box, extentX=1.01, extentY=1.01, extentZ=1.01}]
 dyn_hairFront_l [{shapeType=Capsule, length=5, radius=2.75, shapeName=dyn_hairFront_l_capsule, center=<4.5789 1.3665 -0.5678>, rotation=<rot 90 0 0>}]
 dyn_hairFront_mstr_l [{shapeType=Box, extentX=1.01, extentY=1.01, extentZ=1.01}]
 dyn_hairFront_mstr_r [{shapeType=Box, extentX=1.01, extentY=1.01, extentZ=1.01}]
 dyn_hairFront_r [{shapeType=Capsule, length=5, radius=3, shapeName=dyn_hairFront_r_capsule, center=<5.0152 -1.2101 -0.0125>, rotation=<rot 0 -0.0002 90>}]
 dyn_hairSideHigh_l [{shapeType=Capsule, length=0.505, radius=0.505}]
 dyn_hairSideHigh_r [{shapeType=Capsule, length=0.505, radius=0.505}]
 dyn_hairSide_l [{shapeType=Capsule, length=4.25, radius=3.75, shapeName=dyn_hairSide_l_capsule, center=<3.9119 1.8373 0>, rotation=<rot 0 0 90>}]
 dyn_hairSide_mstr_l [{shapeType=Box, extentX=1.01, extentY=1.01, extentZ=1.01}]
 dyn_hairSide_mstr_r [{shapeType=Box, extentX=1.01, extentY=1.01, extentZ=1.01}]
 dyn_hairSide_r [{shapeType=Capsule, length=8.5, radius=3, shapeName=dyn_hairSide_r_capsule, center=<4.3279 0.1624 0.1399>, rotation=<rot 0 0 90>}]
}
constraints [
 {bone1Name=dyn_hairFront_l, bone2Name=dyn_hairFront_mstr_l, swing1LimitDegrees=15, swing2LimitDegrees=15, twistLimitDegrees=5}
 {bone1Name=dyn_hairFront_r, bone2Name=dyn_hairFront_mstr_r, swing1LimitDegrees=15, swing2LimitDegrees=15, twistLimitDegrees=5}
 {bone1Name=dyn_hairSide_r, bone2Name=dyn_hairSideHigh_r, swing1LimitDegrees=20, swing2LimitDegrees=20, twistLimitDegrees=5}
 {bone1Name=dyn_hairSide_l, bone2Name=dyn_hairSideHigh_l, swing1LimitDegrees=20, swing2LimitDegrees=20, twistLimitDegrees=5}
 {bone1Name=dyn_hairBack_l, bone2Name=dyn_hairBackHigh_l, swing1LimitDegrees=20, swing2LimitDegrees=20, twistLimitDegrees=5}
 {
  bone1Name=dyn_hairSideHigh_r
  bone2Name=dyn_hairSide_mstr_r
  swing1LimitDegrees=45
  swing1Motion=Free
  swing2LimitDegrees=45
  swing2Motion=Free
  twistLimitDegrees=45
  twistMotion=Free
 }
 {
  bone1Name=dyn_hairSideHigh_l
  bone2Name=dyn_hairSide_mstr_l
  swing1LimitDegrees=45
  swing1Motion=Free
  swing2LimitDegrees=45
  swing2Motion=Free
  twistLimitDegrees=45
  twistMotion=Free
 }
 {
  bone1Name=dyn_hairBackHigh_l
  bone2Name=dyn_hairBack_mstr_l
  swing1LimitDegrees=45
  swing1Motion=Free
  swing2LimitDegrees=45
  swing2Motion=Free
  twistLimitDegrees=45
  twistMotion=Free
 }
]
