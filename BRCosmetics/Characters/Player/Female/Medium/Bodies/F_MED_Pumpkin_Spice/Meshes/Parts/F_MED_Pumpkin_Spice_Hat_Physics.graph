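# Physics asset for the hat of the female medium Pumpkin Spice character, defining bodies for its dangling vine and pumpkin piece.
# A six-link vine chain hangs off the head with tight swing and twist limits, loosest at the fourth link.
# The pumpkin piece has tight limits against a small box master body, which rotates freely on the head.
# Head, upper spine and clavicle capsules, linked by free-rotation constraints, supply body volumes around the hat.
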PhysicsAsset /BRCosmetics/Characters/Player/Female/Medium/Bodies/F_MED_Pumpkin_Spice/Meshes/Parts/F_MED_Pumpkin_Spice_Hat_Physics
bodies {
 clavicle_l [{shapeType=Capsule, length=7.75, radius=7.75, shapeName=clavicle_l_capsule, center=<10.5645 -0.0003 0>, rotation=<rot 90 0 0>}]
 clavicle_r [{shapeType=Capsule, length=7.75, radius=7.75, shapeName=clavicle_r_capsule, center=<-7.6112 0 0>, rotation=<rot 90 0 0>}]
 dyn_pumpkinPIece [{shapeType=Capsule, length=6, radius=1.5, shapeName=dyn_pumpkinPIece_capsule, center=<4.2951 -0.2113 -0.1305>, rotation=<rot 90 0 0>}]
 dyn_pumpkinPIece_mstr [{shapeType=Box, extentX=1.01, extentY=1.01, extentZ=1.01}]
 dyn_vine1 [{shapeType=Capsule, length=4.6158, radius=1.3595, center=<3.4768 -0.0813 0.6046>, rotation=<rot 90 0 0>}]
 dyn_vine2 [{shapeType=Capsule, length=4.0794, radius=1.1292, center=<3.1161 0.0391 -0.0675>, rotation=<rot 90 0 0>}]
 dyn_vine3 [{shapeType=Capsule, length=10.1236, radius=2.2142, center=<6.6605 -1.1873 0.2822>, rotation=<rot 84.3748 -0.0006 -0.0006>}]
 dyn_vine4 [
  {shapeType=Sphere, radius=5, shapeName=dyn_vine4_sphere, center=<7.216 -2.2813 1.2746>}
  {shapeType=Capsule, length=8.4104, radius=1.6399, center=<5.8919 -0.2243 -0.3101>, rotation=<rot 90 -169.975 -169.975>}
 ]
 dyn_vine5 [{
   shapeType=Capsule
   length=4.2329
   radius=1.6276
   shapeName=dyn_vine5_capsule
   center=<3.7197 -0.5473 0.0202>
   rotation=<rot 0 81.6345 -89.6917>
  }]
 dyn_vine6 [{shapeType=Capsule, length=3.75, radius=3.25, shapeName=dyn_vine6_capsule, center=<1.9834 -2.2352 1.4877>, rotation=<rot 90 0 0>}]
 head [{shapeType=Capsule, length=0, radius=14.8317, center=<1.5999 -1.6278 -0.6503>, rotation=<rot 0 0 -90>}]
 spine_05 [{shapeType=Capsule, length=48.0834, radius=10.9176, center=<6.2103 0 0>}]
}
constraints [
 {bone1Name=dyn_pumpkinPIece, bone2Name=dyn_pumpkinPIece_mstr, swing1LimitDegrees=10, swing2LimitDegrees=10, twistLimitDegrees=1}
 {bone1Name=dyn_vine1, bone2Name=head, swing1LimitDegrees=5, swing2LimitDegrees=5, twistLimitDegrees=5}
 {
  bone1Name=dyn_pumpkinPIece_mstr
  bone2Name=head
  swing1LimitDegrees=45
  swing1Motion=Free
  swing2LimitDegrees=45
  swing2Motion=Free
  twistLimitDegrees=45
  twistMotion=Free
 }
 {bone1Name=dyn_vine6, bone2Name=dyn_vine5, swing1LimitDegrees=10, swing2LimitDegrees=10, twistLimitDegrees=5}
 {bone1Name=dyn_vine2, bone2Name=dyn_vine1, swing1LimitDegrees=8, swing2LimitDegrees=8, twistLimitDegrees=5}
 {bone1Name=dyn_vine3, bone2Name=dyn_vine2, swing1LimitDegrees=8, swing2LimitDegrees=8, twistLimitDegrees=5}
 {bone1Name=dyn_vine4, bone2Name=dyn_vine3, swing1LimitDegrees=20, swing2LimitDegrees=20, twistLimitDegrees=10}
 {bone1Name=dyn_vine5, bone2Name=dyn_vine4, swing1LimitDegrees=10, swing2LimitDegrees=10, twistLimitDegrees=5}
 {
  bone1Name=head
  bone2Name=spine_05
  swing1LimitDegrees=45
  swing1Motion=Free
  swing2LimitDegrees=45
  swing2Motion=Free
  twistLimitDegrees=45
  twistMotion=Free
 }
 {
  bone1Name=clavicle_l
  bone2Name=spine_05
  swing1LimitDegrees=45
  swing1Motion=Free
  swing2LimitDegrees=45
  swing2Motion=Free
  twistLimitDegrees=45
  twistMotion=Free
 }
 {
  bone1Name=clavicle_r
  bone2Name=spine_05
  swing1LimitDegrees=45
  swing1Motion=Free
  swing2LimitDegrees=45
  swing2Motion=Free
  twistLimitDegrees=45
  twistMotion=Free
 }
]
