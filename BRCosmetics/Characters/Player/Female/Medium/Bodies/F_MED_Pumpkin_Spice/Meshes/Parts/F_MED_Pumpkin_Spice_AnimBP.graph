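# Post-process-style AnimBlueprint for the Pumpkin Spice cosmetic body part, copying the attached parent's pose and adding hat and hair physics.
# On customization load, the Stage1 variant tag sets helmetOn or helmetOff, selecting hat-on hair simulation (RigidBody plus side-strand AnimDynamics) or hat-off hair RigidBody.
# Hat RigidBody always runs first; after the hair stages, a fixed additive dyn_vine4 rotation feeds the character's Control Rig.
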AnimBlueprint /BRCosmetics/Characters/Player/Female/Medium/Bodies/F_MED_Pumpkin_Spice/Meshes/Parts/F_MED_Pumpkin_Spice_AnimBP
target_skeleton=&/F_MED_Pumpkin_Spice_Skeleton
graphs [
 AnimGraph {nodes [
   n1 = RigidBody(ComponentPose=Local To Component(LocalPose=Copy Pose From Mesh() {bUseAttachedParent=true})) {
    bClampLinearTranslationLimitToRefPose=true
    bOverrideWorldGravity=true
    overridePhysicsAsset=&/F_MED_Pumpkin_Spice_Hat_Physics
    simulationSpace=BaseBoneSpace
    alphaBoolBlend {blendOption=Linear}
    baseBoneRef {boneName=spine_05}
    componentAppliedLinearAccClamp=<5000 5000 5000>
    componentLinearAccScale=<0.4 0.4 0.4>
    componentLinearVelScale=<0.3 0.3 0.3>
    overrideWorldGravity {z=-980}
   }
   n2 = Get helmetOn() -> (helmetOn:bool)
   n3 = RigidBody!(ComponentPose=@n1, bAlphaBoolEnabled:bool=@n2, ExternalForce) -> (Pose) {
    alphaInputType=Bool
    bClampLinearTranslationLimitToRefPose=true
    bOverrideWorldGravity=true
    overridePhysicsAsset=&/F_MED_Pumpkin_Spice_Hair_Physics
    simulationSpace=BaseBoneSpace
    alphaBoolBlend {blendOption=Linear}
    baseBoneRef {boneName=spine_05}
    componentAppliedLinearAccClamp=<5000 5000 5000>
    componentLinearAccScale=<0.4 0.4 0.4>
    componentLinearVelScale=<0.3 0.3 0.3>
    overrideWorldGravity {z=-980}
   }
   n4 = Anim Dynamics - Bone: dyn_hairSide_l!(ComponentPose=@n3, bAlphaBoolEnabled:bool=@n2, ExternalForce="0,0,0") -> (Pose) {
    alphaInputType=Bool
    angularDampingOverride=0.9999
    angularSpringConstant=100
    bAngularSpring=true
    bGravityOverrideInSimSpace=true
    bOverrideAngularDamping=true
    simulationSpace=BoneRelative
    alphaBoolBlend {blendOption=Linear}
    boundBone {boneName=dyn_hairSide_l}
    componentAppliedLinearAccClamp=<3500 3500 3500>
    componentLinearAccScale=<1.2 1.2 1.2>
    componentLinearVelScale=<1.2 1.2 1.2>
    physicsBodyDefinitions [{
      sphereCollisionRadius=0
      boundBone {boneName=dyn_hairSide_l}
      boxExtents=<20 10 10>
      constraintSetup {angularLimitsMax=<1 10 15>, angularLimitsMin=<-1 -30 -15>, angularTarget=<0 0 0>}
      localJointOffset=<2 0 0>
     }]
    relativeSpaceBone {boneName=root}
   }
   n5 = Anim Dynamics - Bone: dyn_hairSide_r!(ComponentPose=@n4, bAlphaBoolEnabled:bool=@n2, ExternalForce="0,0,0") -> (Pose) {
    alphaInputType=Bool
    angularDampingOverride=0.9999
    angularSpringConstant=100
    bAngularSpring=true
    bGravityOverrideInSimSpace=true
    bOverrideAngularDamping=true
    simulationSpace=BoneRelative
    alphaBoolBlend {blendOption=Linear}
    boundBone {boneName=dyn_hairSide_r}
    componentAppliedLinearAccClamp=<3500 3500 3500>
    componentLinearAccScale=<1.2 1.2 1.2>
    componentLinearVelScale=<1.2 1.2 1.2>
    physicsBodyDefinitions [{
      sphereCollisionRadius=0
      boundBone {boneName=dyn_hairSide_r}
      boxExtents=<20 10 10>
      constraintSetup {angularLimitsMax=<1 10 15>, angularLimitsMin=<-1 -30 -15>, angularTarget=<0 0 0>}
      localJointOffset=<2 0 0>
     }]
    relativeSpaceBone {boneName=root}
   }
   n6 = RigidBody(ComponentPose=@n5, bAlphaBoolEnabled=Get helmetOff() -> (helmetOff:bool)) {
    alphaInputType=Bool
    bClampLinearTranslationLimitToRefPose=true
    bOverrideWorldGravity=true
    overridePhysicsAsset=&/F_MED_Pumpkin_Spice_HairHatOff_Physics
    simulationSpace=BaseBoneSpace
    alphaBoolBlend {blendOption=Linear}
    baseBoneRef {boneName=spine_05}
    componentAppliedLinearAccClamp=<5000 5000 5000>
    componentLinearAccScale=<0.55 0.4 0.4>
    componentLinearVelScale=<0.3 0.3 0.3>
    overrideWorldGravity {z=-980}
   }
   n7 = "Transform (Modify) Bone - Bone: dyn_vine4"(ComponentPose=@n6, Alpha=1, Translation="0,0,0", Rotation="0,0,11", Scale="1,1,1") {
    rotationMode=BMM_Additive
    translationMode=BMM_Additive
    alphaBoolBlend {blendOption=Linear}
    boneToModify {boneName=dyn_vine4}
    rotation {roll=11}
   }
   n8 = Control Rig(Source=Component To Local(ComponentPose=@n7)) {
    alphaBoolBlend {blendOption=Linear}
    controlRigAssetReference {blueprintRigClass=/BRCosmetics/Characters/Player/Female/Medium/Bodies/F_MED_Pumpkin_Spice/Meshes/Parts/F_MED_Pumpkin_Spice_CtrlRig/F_MED_Pumpkin_Spice_CtrlRig.F_MED_Pumpkin_Spice_CtrlRig_C}
   }
   Output Pose(Result=@n8) {layerGroup=None}
  ]}
 EventGraph {nodes [
   Event OnCharacterCustomizationLoaded()
   > n1 = Branch(Condition=CharacterItemDefinitionHasAnyMetaTag(Tags=Make Array("[0]"=(TagName="Cosmetics.Variant.Property.Stage1"))))
   > Set helmetOn(helmetOn:bool=true) -> (Output_Get:bool)
   Set helmetOff(execute=@n1.else, helmetOff:bool=true) -> (Output_Get:bool)
   Event BlueprintUpdateAnimation()
  ]}
]
variables [
 helmetOff {type=boolean}
 helmetOn {type=boolean}
 Zipline {type=number}
]
